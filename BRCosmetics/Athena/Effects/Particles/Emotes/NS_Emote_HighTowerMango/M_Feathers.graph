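# M_Feathers is a material for feather particles in the HighTowerMango emote's Niagara effect.
# It is a two-sided Default Lit material flagged for Niagara mesh particles, with spherical particle normals enabled.
# Its graph holds a single ParticleColor node, with no textures, parameters or visible connections to material inputs.
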Material /BRCosmetics/Athena/Effects/Particles/Emotes/NS_Emote_HighTowerMango/M_Feathers
graph {
 properties {
  bGenerateSphericalParticleNormals=true
  bUsedWithNiagaraMeshParticles=true
  pixelDepthOffsetMode=PDOM_Legacy
  translucencyLightingMode=TLM_VolumetricPerVertexDirectional
  twoSided=true
  usedShadingModels=Default Lit
 }
 nodes [ParticleColor()]
}
