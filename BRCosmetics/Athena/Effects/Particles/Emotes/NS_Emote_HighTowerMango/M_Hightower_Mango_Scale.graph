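# Translucent lit Niagara particle material for the HighTowerMango emote, drawing hard-edged blue smoke-textured patches near scene geometry.
# Base color blends two blues by a Fresnel on the flattened scene-texture world normal, with a single flat color on forward and mobile paths.
# Opacity multiplies inverted depth fade, a radial UV mask, a sharply thresholded tiling texture and particle alpha.
# Dynamic parameters drive Fresnel exponent, fade distance, tiling, V offset and threshold.
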Material /BRCosmetics/Athena/Effects/Particles/Emotes/NS_Emote_HighTowerMango/M_Hightower_Mango_Scale
graph {
 properties {
  bCastRayTracedShadows=false
  bGenerateSphericalParticleNormals=true
  bUseMaterialAttributes=true
  bUsedWithNiagaraMeshParticles=true
  bUsedWithNiagaraSprites=true
  blendMode=BLEND_Translucent
  pixelDepthOffsetMode=PDOM_Legacy
  translucencyLightingMode=TLM_VolumetricPerVertexDirectional
  usedShadingModels=Default Lit
 }
 nodes [
  ObjectBounds()
  ObjectPositionWS()
  QualitySwitch()
  n1 = VectorParameter() {parameterName=Param_1, defaultValue=<rgba 0.0001 0.0135 0.0573 1>}
  n2 = DynamicParameter() {parameterIndex=1, defaultValue=<rgba 1 1 0 1>, paramNames [Exponent, Target Low, Target High, Depth]}
  n3 = ComponentMask(None=SceneTexture() {desc=Talk to bryce, sceneTextureId=PPI_WorldNormal}.Color) {b=true, g=true, r=true}
  n4 = ScalarParameter() {group=World Normal, parameterName=FlattenNormal}
  n5 = LinearInterpolate(
   A=VectorParameter() {defaultValue=<rgba 0.0451 0.0595 0.3333 1>}.RGB
   B=@n1.RGB
   Alpha=Fresnel(
    ExponentIn=@n2.Exponent
    Normal=MaterialFunctionCall(Normal=@n3, Flatness=@n4) {materialFunction=&FlattenNormal}.Result
   ) {baseReflectFraction=0, exponent=1}
  )
  n6 = DynamicParameter() {defaultValue=<rgba 1 1 0 1>, paramNames [Freq U, Freq V, Offset V, Depth]}
  n7 = TextureCoordinate()
  n8 = Constant() {r=0.5}
  n9 = Multiply(
   A=OneMinus(None=DepthFade(FadeDistance=@n6.Depth) {fadeDistanceDefault=80})
   B=Power(
    Base=SphereMask(A=@n7, B=@n8) {attenuationRadius=0.5, hardnessPercent=0}
    Exp=ScalarParameter() {defaultValue=2, group=Opacity, parameterName=SphereMaskPower, sortPriority=5}
   )
  )
  n10 = DynamicParameter() {defaultValue=<rgba 1 1 0 1>, paramNames [Freq U, Freq V, Offset V, Depth]}
  n11 = MaterialFunctionCall(Float2=Multiply(A=TextureCoordinate(), B=AppendVector(A=@n10.Freq U, B=@n10.Freq U))) {materialFunction=&BreakOutFloat2Components}
  n12 = DynamicParameter() {defaultValue=<rgba 1 1 0 1>, paramNames [Freq U, Freq V, Offset V, Depth]}
  n13 = TextureSampleParameter2D(UVs=AppendVector(A=@n11.R, B=Add(A=@n11.G, B=@n12.Offset V))) {group=Tiling Texture, parameterName=TilingTexture, texture=&T_Inky_Smoke_Tile_Chromatic}
  n14 = Desaturation(None=@n13.RGB) {luminanceFactors=<rgba 1 1 1 0>}
  n15 = DynamicParameter() {parameterIndex=2, defaultValue=<rgba 0 1 0 1>, paramNames [Min, Max, Offset V, Depth]}
  n16 = Saturate(None=Multiply(A=Add(A=@n14, B=@n15.Min), B=Constant() {r=100}))
  SetMaterialAttributes!(
   MaterialAttributes=SetMaterialAttributes!(MaterialAttributes, Base Color=ShadingPathSwitch(Default=@n5, Deferred=@n5, Forward=@n1.RGB, Mobile=@n1.RGB), Emissive Color) {attributeSetTypes [#1, #2]}
   Opacity=Multiply(A=Multiply(A=@n9, B=@n16), B=ParticleColor().A)
  ) {attributeSetTypes [#3]}
  Subtract()
  WorldPosition()
 ]
}
