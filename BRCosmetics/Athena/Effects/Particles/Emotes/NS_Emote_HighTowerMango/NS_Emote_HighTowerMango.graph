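# Niagara system for the HighTowerMango emote, drawing an animated wave of decal sprites and flipping feather meshes across the character's skeletal mesh.
# Decal and two feather emitters (one GPU) burst-spawn partitioned particles, then BoneFlow places them along the User skeletal mesh by angle and radius.
# An emitter-age-driven Wave value fades, scales and flips particles, and kills spent feathers.
# Flash_Decal fires three brightening blue, downward-moving flashes, timed by User.Delay.
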NiagaraSystem /BRCosmetics/Athena/Effects/Particles/Emotes/NS_Emote_HighTowerMango/NS_Emote_HighTowerMango
dependencies {
 usedDataInterfaces [&NiagaraDataInterfaceCurve, &NiagaraDataInterfaceColorCurve, &NiagaraDataInterfaceVector2DCurve]
 usedDynamicInputs []
 usedModules [&SystemState]
}
emitters {
 Decal {
  emitterUpdateScript [
   &EmitterState()
   /Game/Effects/Niagara/Modules/Partition/FN_PartitionAttributes/FN_PartitionAttributes_v2(Particles Per Partition=12, Partitions=5)
   &SpawnBurst_Instantaneous(Spawn Count=$Emitter.Spawn Total, Spawn Probability=HlslExpression() {hlslExpression="Emitter.LoopCount == 0"}, Use Spawn Probability=true)
   /SetVariables(Emitter.T=HlslExpression() {hlslExpression="min(Emitter.Age, 1.0)"}) {isSetParameters=true, module=SetVariables_#1}
   &SpawnRate(
    Spawn Probability=HlslExpression() {hlslExpression="Emitter.T < 1.0"}
    SpawnRate=&MakeFloatFromInt(Int=$Emitter.Spawn Total)
    Use Spawn Probability=true
   ) {enabled=false}
   &self:SpawnContinuous(Count:NiagaraInt32=$Emitter.Spawn Total, Enable Spawn=HlslExpression() {hlslExpression="Emitter.Age < 1.0"}, Every Nth Frame=2) {enabled=false, module=SpawnContinuous}
  ]
  particleSpawnScript [
   /Game/Effects/Niagara/Modules/Partition/FN_ParticlePartition_v2/FN_ParticlePartition_v2()
   /SetVariables(Particles.WaveOffset:NiagaraFloat=&Add_Float(A=0, B=HlslExpression() {hlslExpression="(-.2*rand(Particles.Partition.ID,0,0,0))-.5"})) {isSetParameters=true, module=SetVariables_#2}
   /SetVariables(Particles.SizeMultiplier:NiagaraFloat=$Particles.T) {isSetParameters=true}
   /SetVariables(Particles.T:NiagaraFloat=&Add_Float(
     A=/Game/Effects/Niagara/DynamicInputs/Math/FN_Fit01(Clamp:NiagaraBool=false, Value:NiagaraFloat=$Particles.T, DestMin:NiagaraFloat=-0.2, DestMax:NiagaraFloat=1)
     B=HlslExpression() {hlslExpression="(-1*rand(Particles.Partition.ID,0,0,0))-1"}
    )) {isSetParameters=true}
   /Niagara/Modules/Spawn/Initialization/InitializeParticle(
    Lifetime=10
    Module.Write Color=true
    Module.Write Lifetime=true
    Module.Write Mass=true
    Module.Write Material Random=true
    Module.Write Position=false
    Module.Write Scale=false
    Module.Write SpriteRotation=true
    Module.Write SpriteSize=true
    Sprite Rotation=&UniformRangedFloat(Minimum=0, Maximum=360)
    Sprite Size=&Multiply_Vector2(
     A=&UniformRangedVector2D(Minimum=<42.3533 7.7624>, Maximum=<38.1361 12.4537>)
     B=&Vector2DFromFloat(Value=HlslExpression() {hlslExpression="Particles.Partition.ID > 3 ? 3 : 1.5"})
    )
   )
   /Game/Effects/Niagara/Modules/Math/FN_EvenNumberToSign(Input:NiagaraInt32=$Particles.Partition.ID)
   /Game/Effects/Niagara/Modules/BoneFlow/FN_Initialize_BoneFlow_Attributes_v2()
   /NiagaraFN/Sprite/FN_RandomizeSpriteUV()
  ]
  particleUpdateScript [
   &ParticleState(Kill Particles When Lifetime Has Elapsed=false, Let Infinitely Lived Particles Die When Emitter Deactivates=true)
   /SetVariables(Particles.Wave:NiagaraFloat=/Game/Effects/Niagara/DynamicInputs/Math/FN_Fit(
     Clamp:NiagaraBool=true
     Value:NiagaraFloat=$Particles.Partition.NormalizedID
     SourceMin:NiagaraFloat=HlslExpression() {hlslExpression=Particles.WaveOffset + Emitter.Age*.6 - .1}
     SourceMax:NiagaraFloat=HlslExpression() {hlslExpression=Particles.WaveOffset + Emitter.Age*.6 + .1}
     DestMin:NiagaraFloat=0
     DestMax:NiagaraFloat=1
    )) {isSetParameters=true, module=SetVariables_#3}
   /Game/Effects/Niagara/Modules/BoneFlow/FN_BoneFlow_v3/FN_BoneFlow_v3(
    Angle:NiagaraFloat=&Multiply_Float(A=$Particles.Initial.MaterialRandom, B=6.29)
    Radius=2.1916
    SideVector3:Vector3f=<0 1 0>
    SkeletalMesh:NiagaraDataInterfaceSkeletalMesh=$User.SkeletalMesh
    T:NiagaraFloat=&Multiply_Float(A=$Particles.Initial.Partition.NormalizedID, B=1)
   )
   &ScaleSpriteSize(Scale Factor=&Vector2DFromCurve(
     Vector2Curve=DataInterface() {propertyValues {
       XCurve {keys [{arriveTangent=0.4667, interpMode=RCIM_Cubic, leaveTangent=0.4667, tangentMode=RCTM_User, time=-0.0108, value=0.7371}, <at 1.0042 0.9093>]}
       YCurve {keys [
         {arriveTangent=0.6512, interpMode=RCIM_Cubic, leaveTangent=0.6512, tangentMode=RCTM_User, time=-0.0063, value=0.9265}
         {arriveTangent=1.7426, interpMode=RCIM_Cubic, leaveTangent=1.7426, tangentMode=RCTM_User, time=1.0042, value=2.2003}
        ]}
      }}
     CurveIndex=&OneMinusFloat(Float=$Particles.Wave)
     ScaleCurve=<1 1>
    ))
   &CameraOffset(Camera Offset Amount=80)
   /NiagaraFN/Sprite/FN_SpriteAlignmentAndFacing(SpriteAlignment:Vector3f=$Output.FN_BoneFlow_v3.SplineTangent) {module=SpriteAlignmentAndFacing}
   &Color(Scale Alpha=&FloatFromCurve(FloatCurve=DataInterface() {propertyValues {Curve {keys [<at 0.0003 1.0102>,<at 1 0>]}}}, CurveIndex=$Particles.Wave))
   &ScaleSpriteSize(Scale Factor=&Vector2DFromCurve(
     Vector2Curve=DataInterface() {propertyValues {
       XCurve {keys [{arriveTangent=0.4667, interpMode=RCIM_Cubic, leaveTangent=0.4667, tangentMode=RCTM_User, time=-0.0108, value=0.7371}, <at 1.0042 0.9093>]}
       YCurve {keys [
         {arriveTangent=0.6512, interpMode=RCIM_Cubic, leaveTangent=0.6512, tangentMode=RCTM_User, time=-0.0063, value=0.9265}
         {arriveTangent=1.7426, interpMode=RCIM_Cubic, leaveTangent=1.7426, tangentMode=RCTM_User, time=1.0042, value=2.2003}
        ]}
      }}
     CurveIndex=$Particles.Wave
     ScaleCurve=<1 1>
    )) {module=ScaleSpriteSize001}
   /Game/Effects/Niagara/Modules/Math/FN_Integrator/FN_Integrator_v2(Input=HlslExpression() {hlslExpression="Particles.Wave < .5 ? Engine.DeltaTime : 0"}, Namespace:ENiagaraNameSpace=::NewEnumerator2)
   /Niagara/Modules/Update/SubUV/SubUVAnimation(
    Number Of Frames=32
    SubUV Lookup Index=/Game/Effects/Niagara/DynamicInputs/Math/FN_InterpFloat(
     Interp type:ENiagaraInterpTypes=::NewEnumerator4
     A:NiagaraFloat=&FN_Fit01(Clamp:NiagaraBool=false, Value:NiagaraFloat=$Particles.MaterialRandom, DestMin:NiagaraFloat=0.4, DestMax:NiagaraFloat=0.6)
     B:NiagaraFloat=0.7
     Exponent:NiagaraInt32=3
     Value:NiagaraFloat=$Particles.FN_Integrator_v2.Attr
    )
   )
   &DynamicMaterialParameters(
    Index 0 Param 1=&FN_Fit01(Clamp:NiagaraBool=false, Value:NiagaraFloat=$Particles.MaterialRandom, DestMin:NiagaraFloat=1, DestMax:NiagaraFloat=4)
    Index 0 Param 2=&FN_Fit01(Clamp:NiagaraBool=false, Value:NiagaraFloat=$Particles.MaterialRandom, DestMin:NiagaraFloat=2, DestMax:NiagaraFloat=1)
    Index 0 Param 4=200
    Index 1 Param 1=1
    Index 1 Param 3=1
    Index 1 Param 4=100
    Index 2 Param 1=/Niagara/DynamicInputs/Math/Min_Float(
     A:NiagaraFloat=&FloatFromCurve(
      FloatCurve=DataInterface() {propertyValues {Curve {keys [
          {arriveTangent=0.8032, interpMode=RCIM_Cubic, leaveTangent=0.8032, tangentMode=RCTM_User, time=0.3771, value=-2.0252}
          {arriveTangent=0.323, interpMode=RCIM_Cubic, leaveTangent=0.323, tangentMode=RCTM_User, time=2.6728, value=0.8275}
          <at 3.157 0.8885>
         ]}}}
      CurveIndex=$Emitter.Age
     )
     B:NiagaraFloat=2
    )
    Index 2 Param 2=1
    Index 2 Param 3=0.1
    Module.Param0WriteEnabled=true
    Module.Param0WriteEnabled1=true
    Module.Param0WriteEnabled2=true
    Module.Param1WriteEnabled=true
    Module.Param1WriteEnabled1=true
    Module.Param1WriteEnabled2=true
    Module.Param2WriteEnabled=true
    Module.Param2WriteEnabled1=true
    Module.Param2WriteEnabled2=true
    Module.Param3WriteEnabled=true
    Module.Param3WriteEnabled1=true
    Module.Param3WriteEnabled2=true
    Write Parameter Index 1=true
    Write Parameter Index 2=true
   ) {module=DynamicMaterialParameters001}
   /Game/Effects/Niagara/Modules/Sprite/FN_SpriteRenderer(
    Color:LinearColor=&MultiplyLinearColor(
     Linear Color=&ColorFromCurve(
      ColorCurve=DataInterface() {propertyValues {
        AlphaCurve {keys [<at 0 0 RCIM_Cubic>,<at 0.2478 1>,<at 1 0 RCIM_Cubic>]}
        BlueCurve {keys [<at 0 0.25 RCIM_Cubic>,<at 1 0.1354 RCIM_Cubic>]}
        GreenCurve {keys [<at 0 0.0147 RCIM_Cubic>,<at 1 0.0055 RCIM_Cubic>]}
        RedCurve {keys [<at 0 0 RCIM_Cubic>,<at 1 0.0045 RCIM_Cubic>]}
       }}
      CurveIndex=$Particles.FN_Integrator_v2.Attr
     )
     Scale Factor=&MakeFloat4FromVectorAndFloat(
      "Vector (RGB)"=<1 1 1>
      "Float (Alpha)"=&FloatFromCurve(
       FloatCurve=DataInterface() {propertyValues {Curve {keys [
           {arriveTangent=0.2668, interpMode=RCIM_Cubic, leaveTangent=0.2668, tangentMode=RCTM_User, time=-0.0033, value=0.265}
           {arriveTangent=0.9986, interpMode=RCIM_Cubic, leaveTangent=0.9986, tangentMode=RCTM_User, time=1.4976, value=1}
          ]}}}
       CurveIndex=$Emitter.Age
       Scale Curve=0.5
      )
     )
    )
    Create Color Attribute=true
    Create SpriteSize Attribute=true
    SpriteSize:Vector2f=&Multiply_Vector2(
     A=<50 50>
     B=&Vector2DFromFloat(Value=&FN_Fit01(Clamp:NiagaraBool=true, Value:NiagaraFloat=$Particles.SizeMultiplier, DestMin:NiagaraFloat=1, DestMax:NiagaraFloat=0.5))
    )
   )
  ]
  rendererClasses [&NiagaraSpriteRendererProperties, &NiagaraMeshRendererProperties]
  renderers [
   {rendererClass=&NiagaraSpriteRendererProperties}
   {rendererClass=&NiagaraMeshRendererProperties, rendererIndex=1}
   {rendererClass=&NiagaraSpriteRendererProperties, rendererIndex=2}
   {rendererClass=&NiagaraSpriteRendererProperties, rendererIndex=3}
   {rendererClass=&NiagaraSpriteRendererProperties, rendererIndex=4}
  ]
 }
 Flash_Decal {
  emitterUpdateScript [
   &EmitterState(
    Delay First Loop Only=true
    Life Cycle Mode=::NewEnumerator1
    Loop Behavior=::NewEnumerator2
    Loop Count=3
    Loop Delay=&Subtract_Float(A=$User.Delay, B=0.15)
    Loop Duration=0.15
    UseLoopDelay=true
   )
   &SpawnBurst_Instantaneous(Spawn Count=1, Use Spawn Probability=false)
  ]
  particleSpawnScript [
   &InitializeParticle(
    Color=&MakeLinearColorFromVectorAndFloat(
     "Vector (RGB)"=&MakeVector(
      X=0
      Y=0
      Z=&FN_Fit01(
       Clamp:NiagaraBool=false
       Value:NiagaraFloat=&MakeFloatFromInt(Int=$Emitter.LoopCount)
       DestMin:NiagaraFloat=10
       DestMax:NiagaraFloat=1150
      )
     )
     "Float (Alpha)"=1
    )
    Lifetime=&UniformRangedFloat(Minimum=0.1, Maximum=0.4)
    Module.Write Color=true
    Module.Write Lifetime=true
    Module.Write Mass=true
    Module.Write Material Random=true
    Module.Write Position=false
    Module.Write Scale=false
    Module.Write SpriteRotation=false
    Module.Write SpriteSize=true
    Sprite Size=<55 55>
   )
   /Niagara/Modules/Spawn/Location/SphereLocation(
    MaskSpawnGroup=false
    Module.Override Randomness=false
    Sphere Origin:NiagaraPosition=&Add_Vector(A=/Niagara/DynamicInputs/Helpers/SimulationPosition(), B=<0 0 220>)
    Sphere Radius=10
    Surface Only=false
   )
   &AddVelocity(Velocity=&MakeVector(X=0, Y=0, Z=&UniformAOrBFloat(A=-800, B=-1000)))
  ]
  particleUpdateScript [
   &ParticleState()
   &SolveForcesAndVelocity(Clamp Velocity=false, Limit Acceleration=false)
   &CameraOffset(Camera Offset Amount=100)
   &Color() {enabled=false}
   &ScaleColor(
    Scale Alpha=&FloatFromCurve(FloatCurve=DataInterface() {propertyValues {Curve {keys [<at 0 1>,<at 1 0>]}}})
    ScaleA=true
    ScaleRGB=true
   )
   &KillParticlesInVolume(Kill Shape=::NewEnumerator2, Origin Offset:Vector3f=<0 0 50>)
  ]
  rendererClasses [&NiagaraSpriteRendererProperties]
  renderers [{rendererClass=&NiagaraSpriteRendererProperties}]
 }
 Flipping_Feathers {
  emitterUpdateScript [
   &EmitterState()
   &FN_PartitionAttributes_v2(Particles Per Partition=20, Partitions=2)
   &SpawnBurst_Instantaneous(Spawn Count=$Emitter.Spawn Total, Spawn Probability=HlslExpression() {hlslExpression="Emitter.LoopCount == 0"}, Use Spawn Probability=true)
   /SetVariables(Emitter.T=HlslExpression() {hlslExpression="min(Emitter.Age, 1.0)"}) {isSetParameters=true, module=SetVariables_#1}
   &SpawnRate(
    Spawn Probability=HlslExpression() {hlslExpression="Emitter.T < 1.0"}
    SpawnRate=&MakeFloatFromInt(Int=$Emitter.Spawn Total)
    Use Spawn Probability=true
   ) {enabled=false}
   &self:SpawnContinuous(Count:NiagaraInt32=$Emitter.Spawn Total, Enable Spawn=HlslExpression() {hlslExpression="Emitter.Age < 1.0"}, Every Nth Frame=2) {enabled=false, module=SpawnContinuous}
  ]
  particleSpawnScript [
   &FN_ParticlePartition_v2()
   /SetVariables(Particles.T:NiagaraFloat=&Add_Float(A=$Particles.T, B=HlslExpression() {hlslExpression="(-1*rand(Particles.Partition.ID,0,0,0))-1"})) {isSetParameters=true, module=SetVariables_#4}
   &FN_EvenNumberToSign(Input:NiagaraInt32=$Particles.Partition.ID)
   &InitializeParticle(
    Color=&ColorFromCurve(
     ColorCurve=DataInterface() {propertyValues {
       AlphaCurve {keys [<at 0 1 RCIM_Cubic>,<at 1 1 RCIM_Cubic>]}
       BlueCurve {keys [<at 0 0.3958 RCIM_Cubic>,<at 1 0.1875 RCIM_Cubic>]}
       GreenCurve {keys [<at 0 0.0994 RCIM_Cubic>,<at 1 0.0969 RCIM_Cubic>]}
       RedCurve {keys [<at 0 0.0602 RCIM_Cubic>,<at 1 0.0273 RCIM_Cubic>]}
      }}
     CurveIndex=$Particles.T
    )
    Lifetime=&UniformRangedFloat(Minimum=2, Maximum=2)
    Mesh Scale=&VectorFromFloat(Value=&MakeCustomFloatFromBool(Bool=HlslExpression() {hlslExpression="Particles.Partition.ID > 1 ? 1.0 : 0.0"}, True Float=0.03, False Float=0.02))
    Module.Write Color=true
    Module.Write Lifetime=true
    Module.Write Mass=true
    Module.Write Material Random=true
    Module.Write Position=false
    Module.Write Scale=true
    Module.Write SpriteRotation=false
    Module.Write SpriteSize=false
    Sprite Rotation=&UniformRangedFloat(Minimum=-360, Maximum=360)
    Sprite Size=&Vector2DFromFloat(Value=&UniformRangedFloat(Minimum=4, Maximum=6))
   )
   /SetVariables(Particles.RandomOffset:NiagaraFloat=&FN_Fit01(Clamp:NiagaraBool=false, Value:NiagaraFloat=$Particles.MaterialRandom, DestMin:NiagaraFloat=-0.1, DestMax:NiagaraFloat=0.1)) {isSetParameters=true, module=SetVariables_#5}
   /SetVariables(Particles.WaveOffset:NiagaraFloat=&Add_Float(A=0, B=HlslExpression() {hlslExpression="(-.2*rand(Particles.Partition.ID,0,0,0))-.5+Particles.RandomOffset"})) {isSetParameters=true, module=SetVariables_#2}
   &FN_Initialize_BoneFlow_Attributes_v2(
    Angle=HlslExpression() {hlslExpression=Particles.Partition.NormalizedID*6.28*Particles.Sign*6.28}
    Radius:NiagaraFloat=&self:ScratchDynamicInput(
     arms:NiagaraDataInterfaceCurve=DataInterface() {propertyValues {Curve {keys [{arriveTangent=-3.6758, interpMode=RCIM_Cubic, leaveTangent=-3.6758, tangentMode=RCTM_User, time=-0.0161, value=2.0743}, <at 0.9963 0.2964>]}}}
     legs:NiagaraDataInterfaceCurve=DataInterface() {propertyValues {Curve {keys [
         <at 0.006 3.8932 RCIM_Cubic>
         {arriveTangent=-3.3426, interpMode=RCIM_Cubic, leaveTangent=-3.3426, time=0.5178, value=2.7184}
         {arriveTangent=-4.0817, interpMode=RCIM_Cubic, leaveTangent=-4.0817, time=0.629, value=1.8107}
         {arriveTangent=2.7594, interpMode=RCIM_Cubic, leaveTangent=2.7594, time=0.8829, value=1.2284}
         <at 1.0005 2.8359 RCIM_Cubic>
        ]}}}
     torso:NiagaraDataInterfaceCurve=DataInterface() {propertyValues {Curve {keys [
         <at -0.0032 3.618>
         {arriveTangent=22.1046, interpMode=RCIM_Cubic, leaveTangent=22.1046, time=0.2411, value=5.9102}
         {arriveTangent=17.0646, interpMode=RCIM_Cubic, leaveTangent=17.0646, time=0.3556, value=11.5479}
         {arriveTangent=-5.9483, interpMode=RCIM_Cubic, leaveTangent=-5.9483, time=0.5108, value=10.5119}
         {arriveTangent=3.1211, interpMode=RCIM_Cubic, leaveTangent=3.1211, time=0.839, value=8.6725}
         <at 0.9942 12.0207>
        ]}}}
    )
   )
  ]
  particleUpdateScript [
   &ParticleState(Kill Particles When Lifetime Has Elapsed=false, Let Infinitely Lived Particles Die When Emitter Deactivates=true)
   /Game/Effects/Niagara/Modules/BoneFlow/FN_BoneFlow_v2/FN_BoneFlow_v2(
    Angle:NiagaraFloat=$Particles.Initial.FN_Initialize_BoneFlow_Attributes_v2.Angle
    Radius:NiagaraFloat=$Particles.FN_Initialize_BoneFlow_Attributes_v2.Radius
    SkeletalMesh:NiagaraDataInterfaceSkeletalMesh=$User.SkeletalMesh
    T:NiagaraFloat=$Particles.Initial.Partition.NormalizedID
   )
   /SetVariables(Particles.Wave:NiagaraFloat=&FloatFromCurve(
     FloatCurve=DataInterface() {propertyValues {Curve {keys [<at -0.0115 0.0039 RCIM_Cubic>,<at 1.0027 1.0455 RCIM_Cubic>]}}}
     CurveIndex=&FN_Fit(
      Clamp:NiagaraBool=true
      Value:NiagaraFloat=$Particles.Partition.NormalizedID
      SourceMin:NiagaraFloat=HlslExpression() {hlslExpression=Particles.WaveOffset + Emitter.Age*.6 - .2}
      SourceMax:NiagaraFloat=HlslExpression() {hlslExpression=Particles.WaveOffset + Emitter.Age*.6 + .2}
      DestMin:NiagaraFloat=0
      DestMax:NiagaraFloat=1
     )
    )) {isSetParameters=true, module=SetVariables_#3}
   /SetVariables(Particles.VisibilityTag=HlslExpression() {hlslExpression="int(Particles.Wave == 1 ? 0 : 1)"}) {isSetParameters=true, module=SetVariables_#6}
   /SetVariables(Particles.MeshOrientation=&MultiplyQuaternion(
     Quaternion A=$Output.Orientation
     Quaternion B=&MakeQuatFromAxisAngle(Axis=<0 1 0>, AngleInDegrees=HlslExpression() {hlslExpression=(Particles.Wave*-180)})
    )) {isSetParameters=true, module=SetVariables_#7}
   &KillParticles(Kill Particles=HlslExpression() {hlslExpression="Particles.Wave ==0 ? 1 : 0"})
   &ScaleMeshSize(Scale Factor=&VectorFromFloat(Value=&FloatFromCurve(
      FloatCurve=DataInterface() {propertyValues {Curve {keys [
          <at -0.0013 0 RCIM_Cubic>
          {arriveTangent=0.856, interpMode=RCIM_Cubic, leaveTangent=0.856, time=0.1419, value=0.8243}
          {arriveTangent=-0.897, interpMode=RCIM_Cubic, leaveTangent=-0.897, time=0.9475, value=0.8122}
          <at 1.0708 -0.0089>
         ]}}}
      CurveIndex=$Particles.Wave
      Scale Curve=0.65
     )))
   /SetVariables(Particles.RenderScale:Vector3f=&Multiply_Vector(
     A=$Particles.Scale
     B=&MakeVector(
      X=&FloatFromCurve(
       FloatCurve=DataInterface() {propertyValues {Curve {keys [
           {arriveTangent=-1.5288, interpMode=RCIM_Cubic, leaveTangent=-1.5288, tangentMode=RCTM_User, time=-0.0116, value=0.937}
           {arriveTangent=2.1033, interpMode=RCIM_Cubic, leaveTangent=2.1033, tangentMode=RCTM_User, time=0.1853, value=1.571}
           {arriveTangent=-1.3425, interpMode=RCIM_Cubic, leaveTangent=-1.3425, tangentMode=RCTM_User, time=0.5237, value=1.6414}
           {arriveTangent=0.0461, interpMode=RCIM_Cubic, leaveTangent=0.0461, tangentMode=RCTM_User, time=0.8692, value=1.0095}
          ]}}}
       CurveIndex=$Particles.Wave
       Scale Curve=0.4
      )
      Y=1
      Z=1
     )
    )) {isSetParameters=true, module=SetVariables_#8}
  ]
  rendererClasses [&NiagaraSpriteRendererProperties, &NiagaraMeshRendererProperties]
  renderers [
   {rendererClass=&NiagaraSpriteRendererProperties}
   {rendererClass=&NiagaraMeshRendererProperties, rendererIndex=1}
  ]
 }
 Flipping_Feathers001 {
  simTarget=GPUComputeSim
  emitterUpdateScript [
   &EmitterState()
   &FN_PartitionAttributes_v2(Particles Per Partition=60, Partitions=4)
   &SpawnBurst_Instantaneous(Spawn Count=$Emitter.Spawn Total, Spawn Probability=HlslExpression() {hlslExpression="Emitter.LoopCount == 0"}, Use Spawn Probability=true)
   /SetVariables(Emitter.T=HlslExpression() {hlslExpression="min(Emitter.Age, 1.0)"}) {isSetParameters=true, module=SetVariables_#1}
   &SpawnRate(
    Spawn Probability=HlslExpression() {hlslExpression="Emitter.T < 1.0"}
    SpawnRate=&MakeFloatFromInt(Int=$Emitter.Spawn Total)
    Use Spawn Probability=true
   ) {enabled=false}
   &self:SpawnContinuous(Count:NiagaraInt32=$Emitter.Spawn Total, Enable Spawn=HlslExpression() {hlslExpression="Emitter.Age < 1.0"}, Every Nth Frame=2) {enabled=false, module=SpawnContinuous}
  ]
  particleSpawnScript [
   &FN_ParticlePartition_v2()
   /SetVariables(Particles.T:NiagaraFloat=&Add_Float(A=$Particles.T, B=HlslExpression() {hlslExpression="(-1*rand(Particles.Partition.ID,0,0,0))-1"})) {isSetParameters=true, module=SetVariables_#4}
   &FN_EvenNumberToSign(Input:NiagaraInt32=$Particles.Partition.ID)
   &InitializeParticle(
    Color=&ColorFromCurve(
     ColorCurve=DataInterface() {propertyValues {
       AlphaCurve {keys [<at 0 1 RCIM_Cubic>,<at 1 1 RCIM_Cubic>]}
       BlueCurve {keys [<at 0 0.3958 RCIM_Cubic>,<at 1 0.1875 RCIM_Cubic>]}
       GreenCurve {keys [<at 0 0.0994 RCIM_Cubic>,<at 1 0.0969 RCIM_Cubic>]}
       RedCurve {keys [<at 0 0.0602 RCIM_Cubic>,<at 1 0.0273 RCIM_Cubic>]}
      }}
     CurveIndex=$Particles.T
    )
    Lifetime=&UniformRangedFloat(Minimum=2, Maximum=2)
    Mesh Scale=&VectorFromFloat(Value=&MakeCustomFloatFromBool(Bool=HlslExpression() {hlslExpression="Particles.Partition.ID > 1 ? 1.0 : 0.0"}, True Float=0.03, False Float=0.02))
    Module.Write Color=true
    Module.Write Lifetime=true
    Module.Write Mass=true
    Module.Write Material Random=true
    Module.Write Position=false
    Module.Write Scale=true
    Module.Write SpriteRotation=false
    Module.Write SpriteSize=false
    Sprite Rotation=&UniformRangedFloat(Minimum=-360, Maximum=360)
    Sprite Size=&Vector2DFromFloat(Value=&UniformRangedFloat(Minimum=4, Maximum=6))
   )
   /SetVariables(Particles.RandomOffset:NiagaraFloat=&FN_Fit01(Clamp:NiagaraBool=false, Value:NiagaraFloat=$Particles.MaterialRandom, DestMin:NiagaraFloat=-0.1, DestMax:NiagaraFloat=0.1)) {isSetParameters=true, module=SetVariables_#5}
   /SetVariables(Particles.WaveOffset:NiagaraFloat=&Add_Float(A=0, B=HlslExpression() {hlslExpression="(-.2*rand(Particles.Partition.ID,0,0,0))-.5+Particles.RandomOffset"})) {isSetParameters=true, module=SetVariables_#2}
   &FN_Initialize_BoneFlow_Attributes_v2(
    Angle=HlslExpression() {hlslExpression=Particles.Partition.NormalizedID*6.28*Particles.Sign*6.28}
    Radius:NiagaraFloat=&self:ScratchDynamicInput(
     arms:NiagaraDataInterfaceCurve=DataInterface() {propertyValues {Curve {keys [{arriveTangent=-3.6758, interpMode=RCIM_Cubic, leaveTangent=-3.6758, tangentMode=RCTM_User, time=-0.0161, value=2.0743}, <at 0.9963 0.2964>]}}}
     legs:NiagaraDataInterfaceCurve=DataInterface() {propertyValues {Curve {keys [
         <at 0.006 3.8932 RCIM_Cubic>
         {arriveTangent=-3.3426, interpMode=RCIM_Cubic, leaveTangent=-3.3426, time=0.5178, value=2.7184}
         {arriveTangent=-4.0817, interpMode=RCIM_Cubic, leaveTangent=-4.0817, time=0.629, value=1.8107}
         {arriveTangent=2.7594, interpMode=RCIM_Cubic, leaveTangent=2.7594, time=0.8829, value=1.2284}
         <at 1.0005 2.8359 RCIM_Cubic>
        ]}}}
     torso:NiagaraDataInterfaceCurve=DataInterface() {propertyValues {Curve {keys [
         <at -0.0032 3.618>
         {arriveTangent=22.1046, interpMode=RCIM_Cubic, leaveTangent=22.1046, time=0.2411, value=5.9102}
         {arriveTangent=17.0646, interpMode=RCIM_Cubic, leaveTangent=17.0646, time=0.3556, value=11.5479}
         {arriveTangent=-5.9483, interpMode=RCIM_Cubic, leaveTangent=-5.9483, time=0.5108, value=10.5119}
         {arriveTangent=3.1211, interpMode=RCIM_Cubic, leaveTangent=3.1211, time=0.839, value=8.6725}
         <at 0.9942 12.0207>
        ]}}}
    )
   )
  ]
  particleUpdateScript [
   &ParticleState(Kill Particles When Lifetime Has Elapsed=false, Let Infinitely Lived Particles Die When Emitter Deactivates=true)
   &FN_BoneFlow_v2(
    Angle:NiagaraFloat=$Particles.Initial.FN_Initialize_BoneFlow_Attributes_v2.Angle
    Radius:NiagaraFloat=$Particles.FN_Initialize_BoneFlow_Attributes_v2.Radius
    SkeletalMesh:NiagaraDataInterfaceSkeletalMesh=$User.SkeletalMesh
    T:NiagaraFloat=$Particles.Initial.Partition.NormalizedID
   )
   /SetVariables(Particles.Wave:NiagaraFloat=&FloatFromCurve(
     FloatCurve=DataInterface() {propertyValues {Curve {keys [<at -0.0115 0.0039 RCIM_Cubic>,<at 1.0027 1.0455 RCIM_Cubic>]}}}
     CurveIndex=&FN_Fit(
      Clamp:NiagaraBool=true
      Value:NiagaraFloat=$Particles.Partition.NormalizedID
      SourceMin:NiagaraFloat=HlslExpression() {hlslExpression=Particles.WaveOffset + Emitter.Age*.6 - .2}
      SourceMax:NiagaraFloat=HlslExpression() {hlslExpression=Particles.WaveOffset + Emitter.Age*.6 + .2}
      DestMin:NiagaraFloat=0
      DestMax:NiagaraFloat=1
     )
    )) {isSetParameters=true, module=SetVariables_#3}
   /SetVariables(Particles.VisibilityTag=HlslExpression() {hlslExpression="int(Particles.Wave == 1 ? 0 : 1)"}) {isSetParameters=true, module=SetVariables_#6}
   /SetVariables(Particles.MeshOrientation=&MultiplyQuaternion(
     Quaternion A=$Output.Orientation
     Quaternion B=&MakeQuatFromAxisAngle(Axis=<0 1 0>, AngleInDegrees=HlslExpression() {hlslExpression=(Particles.Wave*-180)})
    )) {isSetParameters=true, module=SetVariables_#7}
   &KillParticles(Kill Particles=HlslExpression() {hlslExpression="Particles.Wave ==0 ? 1 : 0"})
   &ScaleMeshSize(Scale Factor=&VectorFromFloat(Value=&FloatFromCurve(
      FloatCurve=DataInterface() {propertyValues {Curve {keys [
          <at -0.0013 0 RCIM_Cubic>
          {arriveTangent=0.856, interpMode=RCIM_Cubic, leaveTangent=0.856, time=0.1419, value=0.8243}
          {arriveTangent=-0.897, interpMode=RCIM_Cubic, leaveTangent=-0.897, time=0.9475, value=0.8122}
          <at 1.0708 -0.0089>
         ]}}}
      CurveIndex=$Particles.Wave
      Scale Curve=0.65
     )))
   /SetVariables(Particles.RenderScale:Vector3f=&Multiply_Vector(
     A=$Particles.Scale
     B=&MakeVector(
      X=&FloatFromCurve(
       FloatCurve=DataInterface() {propertyValues {Curve {keys [
           {arriveTangent=-1.5288, interpMode=RCIM_Cubic, leaveTangent=-1.5288, tangentMode=RCTM_User, time=-0.0116, value=0.937}
           {arriveTangent=2.1033, interpMode=RCIM_Cubic, leaveTangent=2.1033, tangentMode=RCTM_User, time=0.1853, value=1.571}
           {arriveTangent=-1.3425, interpMode=RCIM_Cubic, leaveTangent=-1.3425, tangentMode=RCTM_User, time=0.5237, value=1.6414}
           {arriveTangent=0.0461, interpMode=RCIM_Cubic, leaveTangent=0.0461, tangentMode=RCTM_User, time=0.8692, value=1.0095}
          ]}}}
       CurveIndex=$Particles.Wave
       Scale Curve=0.4
      )
      Y=1
      Z=1
     )
    )) {isSetParameters=true, module=SetVariables_#8}
  ]
  rendererClasses [&NiagaraSpriteRendererProperties, &NiagaraMeshRendererProperties]
  renderers [
   {rendererClass=&NiagaraSpriteRendererProperties}
   {rendererClass=&NiagaraMeshRendererProperties, rendererIndex=1}
  ]
 }
}
summary {
 systemName=NS_Emote_HighTowerMango
 emitters [
  {emitterName=Flash_Decal, rendererClasses [&NiagaraSpriteRendererProperties]}
  {
   emitterName=Flipping_Feathers
   rendererClasses [&NiagaraSpriteRendererProperties, &NiagaraMeshRendererProperties]
  }
  {emitterName=Decal, rendererClasses [&NiagaraSpriteRendererProperties, &NiagaraMeshRendererProperties]}
  {
   emitterName=Flipping_Feathers001
   simTarget=GPUComputeSim
   rendererClasses [&NiagaraSpriteRendererProperties, &NiagaraMeshRendererProperties]
  }
 ]
 userVariables [
  User.BKSL_CurrentLevel {type=NiagaraFloat, defaultValue=160, description=""}
  User.BKSL_StartAtLevel {type=NiagaraFloat, defaultValue=120, description=""}
  User.BKSL_StopAtLevel {type=NiagaraFloat, defaultValue=160, description=""}
  User.BKSL_VariantColor {type=LinearColor, description="", defaultValue {struct=LinearColor, value=<rgba 1 1 1 1>}}
  User.Delay {type=NiagaraFloat, defaultValue=3.33, description=""}
  User.SkeletalMesh {
   type=NiagaraDataInterfaceSkeletalMesh
   defaultValue {
    struct=NiagaraExt_VariableValue_DataInterface
    value {dataInterface=&self:NiagaraDataInterfaceSkeletalMesh_0}
   }
  }
 ]
}
