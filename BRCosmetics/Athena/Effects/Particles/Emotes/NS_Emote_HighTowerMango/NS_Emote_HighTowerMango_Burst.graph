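# Niagara burst effect for the HighTowerMango emote that puffs blue-to-dark smoke sprites off a character's skeletal mesh.
# One short-lived HeavySmoke emitter and two longer-lived, near-identical LightSmoke emitters burst partitioned particles, placed along the User.SkeletalMesh by BoneFlow.
# The partition T value drives bone position, tint, size and upward velocity, added to an outward binormal push.
# Drag, SubUV animation and curves shape each puff's motion, growth and fade.
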NiagaraSystem /BRCosmetics/Athena/Effects/Particles/Emotes/NS_Emote_HighTowerMango/NS_Emote_HighTowerMango_Burst
dependencies {
 usedDataInterfaces [&NiagaraDataInterfaceCurve, &NiagaraDataInterfaceArrayFloat3]
 usedDynamicInputs []
 usedModules [
  &SystemState
  &self:SystemSpawnScript.SystemScriptSource.SystemScriptGraph.NiagaraNodeAssignment_0.SetVariables_#1
 ]
}
emitters {
 HeavySmoke {
  emitterUpdateScript [
   &EmitterState(Inactive Response=::NewEnumerator1, Life Cycle Mode=::NewEnumerator1, Loop Behavior=::NewEnumerator1, Loop Duration=3, UseLoopDelay=true)
   /Game/Effects/Niagara/Modules/Partition/FN_PartitionAttributes/FN_PartitionAttributes_v2(Particles Per Partition=5, Partitions=5)
   &SpawnBurst_Instantaneous(Spawn Count=$Emitter.Spawn Total, Use Spawn Probability=false)
  ]
  particleSpawnScript [
   /Game/Effects/Niagara/Modules/Partition/FN_ParticlePartition_v2/FN_ParticlePartition_v2()
   /Niagara/Modules/Spawn/Initialization/InitializeParticle(
    Color=&LerpLinearColors(StartColor=<rgba 0.1175 0.109 1 1.5>, EndColor=<rgba 0.0125 0.0278 0.0781 1.5>, LerpFactor=$Particles.T)
    Lifetime=&UniformRangedFloat(Minimum=0.8, Maximum=1)
    Module.Write Color=true
    Module.Write Lifetime=true
    Module.Write Mass=true
    Module.Write Material Random=true
    Module.Write Position=false
    Module.Write Scale=false
    Module.Write SpriteRotation=true
    Module.Write SpriteSize=true
    Sprite Rotation=&UniformRangedFloat(Minimum=0, Maximum=360)
    Sprite Size=&Vector2DFromFloat(Value=&Multiply_Float(
      A=&UniformRangedFloat(Minimum=40, Maximum=40)
      B=&FloatFromCurve(FloatCurve=DataInterface() {propertyValues {Curve {keys [<at 0 1>,<at 0.9907 0.6615>]}}}, CurveIndex=$Particles.T)
     ))
   )
   /Game/Effects/Niagara/Modules/BoneFlow/FN_Initialize_BoneFlow_Attributes_v2(T:NiagaraFloat=/Game/Effects/Niagara/DynamicInputs/Math/FN_Fit01(Clamp:NiagaraBool=false, Value:NiagaraFloat=$Particles.T, DestMin:NiagaraFloat=-0.1, DestMax:NiagaraFloat=1.324))
   /NiagaraFN/Sprite/FN_RandomizeSpriteUV()
   /Game/Effects/Niagara/Modules/BoneFlow/FN_BoneFlow_v2/FN_BoneFlow_v2(
    Angle:NiagaraFloat=&Multiply_Float(A=$Particles.MaterialRandom, B=6)
    Radius=0.1364
    SkeletalMesh:NiagaraDataInterfaceSkeletalMesh=$User.SkeletalMesh
    T:NiagaraFloat=$Particles.Initial.FN_Initialize_BoneFlow_Attributes_v2.T
   )
   &AddVelocity(Velocity=&Add_Vector(
     A=&Multiply_VectorByFloat(Vector=$Output.FN_BoneFlow_v2.RotatedBiNormal, Float=50)
     B=&MakeVector(X=0, Y=0, Z=&Multiply_Float(A=&OneMinusFloat(Float=$Particles.T), B=40))
    ))
  ]
  particleUpdateScript [
   &ParticleState(Kill Particles When Lifetime Has Elapsed=false, Let Infinitely Lived Particles Die When Emitter Deactivates=true)
   &Drag(Drag=2, Use Linear Drag=true, Use Rotational Drag=false)
   &SolveForcesAndVelocity(Clamp Velocity=false, Limit Acceleration=false)
   /Niagara/Modules/Update/SubUV/SubUVAnimation(
    Number Of Frames=64
    SubUV Lookup Index=&FloatFromCurve(
     FloatCurve=DataInterface() {propertyValues {Curve {keys [
         <at -0.0062 0.3391>
         {arriveTangent=0.5669, interpMode=RCIM_Cubic, leaveTangent=0.5669, tangentMode=RCTM_User, time=0.0716, value=0.8311}
         <at 0.9986 1.0243>
        ]}}}
     CurveIndex=&FN_Fit01(Clamp:NiagaraBool=false, Value:NiagaraFloat=$Particles.MaterialRandom, DestMin:NiagaraFloat=0.4, DestMax:NiagaraFloat=0.6)
    )
   )
   &CameraOffset(Camera Offset Amount=25)
   &Color(
    Color=$Particles.Initial.Color
    Scale Alpha=&FloatFromCurve(FloatCurve=DataInterface() {propertyValues {Curve {keys [
         <at -0.0038 0.0215 RCIM_Cubic>
         {arriveTangent=0.4882, interpMode=RCIM_Cubic, leaveTangent=0.4882, time=0.1452, value=1.01}
         <at 0.6006 0.3165>
         <at 1 0>
        ]}}})
   )
   &ScaleSpriteSize(Scale Factor=&Vector2DFromFloat(Value=&FloatFromCurve(FloatCurve=DataInterface() {propertyValues {Curve {keys [
          {arriveTangent=2.5063, interpMode=RCIM_Cubic, leaveTangent=2.5063, tangentMode=RCTM_User, time=-0.0008, value=0.6394}
          {arriveTangent=0.2264, interpMode=RCIM_Cubic, leaveTangent=0.2264, tangentMode=RCTM_User, time=0.07, value=1.7344}
          <at 1.0062 1.9434 RCIM_Cubic>
         ]}}})))
   /Game/Effects/Niagara/Modules/Sprite/FN_SpriteRenderer(
    Color:LinearColor=&MultiplyLinearColor(
     Linear Color=$Particles.Initial.Color
     Scale Factor=&MakeFloat4FromVectorAndFloat(
      "Vector (RGB)"=<1 1 1>
      "Float (Alpha)"=&FloatFromCurve(FloatCurve=DataInterface() {propertyValues {Curve {keys [
           <at -0.0038 0.0215 RCIM_Cubic>
           {arriveTangent=0.162, interpMode=RCIM_Cubic, leaveTangent=0.162, tangentMode=RCTM_User, time=0.1558, value=0.7168}
           <at 0.6367 0.29>
           <at 1 0>
          ]}}})
     )
    )
    Create Color Attribute=true
    Create SpriteSize Attribute=true
    SpriteSize:Vector2f=&Multiply_Vector2(A=$Particles.SpriteSize, B=&Vector2DFromFloat(Value=0.8))
   )
   &FN_SpriteRenderer(Camera Offset=100, Create Camera Offset Attribute=true, Create Sprite Rotation Attribute=true) {module=FN_SpriteRenderer001}
  ]
  rendererClasses [&NiagaraSpriteRendererProperties]
  renderers [
   {rendererClass=&NiagaraSpriteRendererProperties}
   {rendererClass=&NiagaraSpriteRendererProperties, rendererIndex=1}
   {rendererClass=&NiagaraSpriteRendererProperties, rendererIndex=2}
   {rendererClass=&NiagaraSpriteRendererProperties, rendererIndex=3}
  ]
 }
 LightSmoke {
  emitterUpdateScript [
   &EmitterState(Inactive Response=::NewEnumerator1, Life Cycle Mode=::NewEnumerator1, Loop Behavior=::NewEnumerator1, Loop Duration=3, UseLoopDelay=true)
   &FN_PartitionAttributes_v2(Particles Per Partition=4, Partitions=4)
   &SpawnBurst_Instantaneous(Spawn Count=$Emitter.Spawn Total, Use Spawn Probability=false)
  ]
  particleSpawnScript [
   &FN_ParticlePartition_v2()
   &InitializeParticle(
    Color=&LerpLinearColors(StartColor=<rgba 0.076 0.0671 1 0.5>, EndColor=<rgba 0 0 0 0.5>, LerpFactor=$Particles.T)
    Lifetime=&UniformRangedFloat(Minimum=4, Maximum=5)
    Module.Write Color=true
    Module.Write Lifetime=true
    Module.Write Mass=true
    Module.Write Material Random=true
    Module.Write Position=false
    Module.Write Scale=false
    Module.Write SpriteRotation=true
    Module.Write SpriteSize=true
    Sprite Rotation=&UniformRangedFloat(Minimum=0, Maximum=360)
    Sprite Size=&Vector2DFromFloat(Value=&Multiply_Float(
      A=&UniformRangedFloat(Minimum=45, Maximum=65)
      B=&FloatFromCurve(FloatCurve=DataInterface() {propertyValues {Curve {keys [<at 0 1>,<at 0.9907 0.6615>]}}}, CurveIndex=$Particles.T, Scale Curve=0.8)
     ))
   )
   &FN_Initialize_BoneFlow_Attributes_v2(T:NiagaraFloat=&FN_Fit01(Clamp:NiagaraBool=false, Value:NiagaraFloat=$Particles.T, DestMin:NiagaraFloat=-0.2, DestMax:NiagaraFloat=1.324))
   &FN_RandomizeSpriteUV()
   &FN_BoneFlow_v2(
    Angle:NiagaraFloat=&Multiply_Float(A=$Particles.MaterialRandom, B=6)
    Radius=0.1364
    SkeletalMesh:NiagaraDataInterfaceSkeletalMesh=$User.SkeletalMesh
    T:NiagaraFloat=$Particles.Initial.FN_Initialize_BoneFlow_Attributes_v2.T
   )
   &AddVelocity(Velocity=&Add_Vector(
     A=&Multiply_VectorByFloat(Vector=$Output.FN_BoneFlow_v2.RotatedBiNormal, Float=50)
     B=&MakeVector(X=0, Y=0, Z=&Multiply_Float(A=&OneMinusFloat(Float=$Particles.T), B=80))
    ))
  ]
  particleUpdateScript [
   &ParticleState()
   &Drag(Drag=3, Use Linear Drag=true, Use Rotational Drag=false)
   &SolveForcesAndVelocity(Clamp Velocity=false, Limit Acceleration=false)
   &SubUVAnimation(
    Number Of Frames=64
    SubUV Lookup Index=&FloatFromCurve(FloatCurve=DataInterface() {propertyValues {Curve {keys [<at -0.0012 0.0729>, {arriveTangent=0.4668, interpMode=RCIM_Cubic, leaveTangent=0.4668, time=0.0554, value=0.3163}, <at 0.9981 0.5394>]}}})
   )
   &CameraOffset(Camera Offset Amount=71.7163)
   &Color(
    Color=$Particles.Initial.Color
    Scale Alpha=&FloatFromCurve(FloatCurve=DataInterface() {propertyValues {Curve {keys [
         <at -0.0038 0.0215 RCIM_Cubic>
         {arriveTangent=0.4882, interpMode=RCIM_Cubic, leaveTangent=0.4882, time=0.1452, value=1.01}
         <at 0.6006 0.3165>
         <at 1 0>
        ]}}})
   )
   &ScaleSpriteSize(Scale Factor=&Vector2DFromFloat(Value=&FloatFromCurve(FloatCurve=DataInterface() {propertyValues {Curve {keys [
          {arriveTangent=2.5063, interpMode=RCIM_Cubic, leaveTangent=2.5063, tangentMode=RCTM_User, time=-0.0008, value=0.6394}
          {arriveTangent=0.2264, interpMode=RCIM_Cubic, leaveTangent=0.2264, tangentMode=RCTM_User, time=0.07, value=1.7344}
          <at 1.0062 1.9434 RCIM_Cubic>
         ]}}})))
   &FN_SpriteRenderer(
    Color:LinearColor=&MultiplyLinearColor(
     Linear Color=$Particles.Initial.Color
     Scale Factor=&MakeFloat4FromVectorAndFloat(
      "Vector (RGB)"=<1 1 1>
      "Float (Alpha)"=&FloatFromCurve(FloatCurve=DataInterface() {propertyValues {Curve {keys [
           <at -0.0038 0.0215 RCIM_Cubic>
           {arriveTangent=0.162, interpMode=RCIM_Cubic, leaveTangent=0.162, tangentMode=RCTM_User, time=0.1558, value=0.7168}
           <at 0.6367 0.29>
           <at 1 0>
          ]}}})
     )
    )
    Create Color Attribute=true
    Create SpriteSize Attribute=true
    SpriteSize:Vector2f=&Multiply_Vector2(A=$Particles.SpriteSize, B=&Vector2DFromFloat(Value=0.8))
   )
   &FN_SpriteRenderer(Camera Offset=100, Create Camera Offset Attribute=true, Create Sprite Rotation Attribute=true) {module=FN_SpriteRenderer001}
  ]
  rendererClasses [&NiagaraSpriteRendererProperties]
  renderers [
   {rendererClass=&NiagaraSpriteRendererProperties}
   {rendererClass=&NiagaraSpriteRendererProperties, rendererIndex=1}
   {rendererClass=&NiagaraSpriteRendererProperties, rendererIndex=2}
   {rendererClass=&NiagaraSpriteRendererProperties, rendererIndex=3}
  ]
 }
 LightSmoke001 {
  emitterUpdateScript [
   &EmitterState(Inactive Response=::NewEnumerator1, Life Cycle Mode=::NewEnumerator1, Loop Behavior=::NewEnumerator1, Loop Duration=3, UseLoopDelay=true)
   &FN_PartitionAttributes_v2(Particles Per Partition=2, Partitions=4)
   &SpawnBurst_Instantaneous(Spawn Count=$Emitter.Spawn Total, Use Spawn Probability=false)
  ]
  particleSpawnScript [
   &FN_ParticlePartition_v2()
   &InitializeParticle(
    Color=&LerpLinearColors(StartColor=<rgba 0.076 0.0671 1 0.5>, EndColor=<rgba 0 0 0 0.5>, LerpFactor=$Particles.T)
    Lifetime=&UniformRangedFloat(Minimum=4, Maximum=5)
    Module.Write Color=true
    Module.Write Lifetime=true
    Module.Write Mass=true
    Module.Write Material Random=true
    Module.Write Position=false
    Module.Write Scale=false
    Module.Write SpriteRotation=true
    Module.Write SpriteSize=true
    Sprite Rotation=&UniformRangedFloat(Minimum=0, Maximum=360)
    Sprite Size=&Vector2DFromFloat(Value=&Multiply_Float(
      A=&UniformRangedFloat(Minimum=45, Maximum=65)
      B=&FloatFromCurve(FloatCurve=DataInterface() {propertyValues {Curve {keys [<at 0 1>,<at 0.9907 0.6615>]}}}, CurveIndex=$Particles.T, Scale Curve=0.8)
     ))
   )
   &FN_Initialize_BoneFlow_Attributes_v2(T:NiagaraFloat=&FN_Fit01(Clamp:NiagaraBool=false, Value:NiagaraFloat=$Particles.T, DestMin:NiagaraFloat=-0.2, DestMax:NiagaraFloat=1.324))
   &FN_RandomizeSpriteUV()
   &FN_BoneFlow_v2(
    Angle:NiagaraFloat=&Multiply_Float(A=$Particles.MaterialRandom, B=6)
    Radius=0.1364
    SkeletalMesh:NiagaraDataInterfaceSkeletalMesh=$User.SkeletalMesh
    T:NiagaraFloat=$Particles.Initial.FN_Initialize_BoneFlow_Attributes_v2.T
   )
   &AddVelocity(Velocity=&Add_Vector(
     A=&Multiply_VectorByFloat(Vector=$Output.FN_BoneFlow_v2.RotatedBiNormal, Float=50)
     B=&MakeVector(X=0, Y=0, Z=&Multiply_Float(A=&OneMinusFloat(Float=$Particles.T), B=80))
    ))
  ]
  particleUpdateScript [
   &ParticleState()
   &Drag(Drag=3, Use Linear Drag=true, Use Rotational Drag=false)
   &SolveForcesAndVelocity(Clamp Velocity=false, Limit Acceleration=false)
   &SubUVAnimation(
    Number Of Frames=64
    SubUV Lookup Index=&FloatFromCurve(FloatCurve=DataInterface() {propertyValues {Curve {keys [<at -0.0012 0.0729>, {arriveTangent=0.4668, interpMode=RCIM_Cubic, leaveTangent=0.4668, time=0.0554, value=0.3163}, <at 0.9981 0.5394>]}}})
   )
   &CameraOffset(Camera Offset Amount=71.7163)
   &Color(
    Color=$Particles.Initial.Color
    Scale Alpha=&FloatFromCurve(FloatCurve=DataInterface() {propertyValues {Curve {keys [
         <at -0.0038 0.0215 RCIM_Cubic>
         {arriveTangent=0.4882, interpMode=RCIM_Cubic, leaveTangent=0.4882, time=0.1452, value=1.01}
         <at 0.6006 0.3165>
         <at 1 0>
        ]}}})
   )
   &ScaleSpriteSize(Scale Factor=&Vector2DFromFloat(Value=&FloatFromCurve(FloatCurve=DataInterface() {propertyValues {Curve {keys [
          {arriveTangent=2.5063, interpMode=RCIM_Cubic, leaveTangent=2.5063, tangentMode=RCTM_User, time=-0.0008, value=0.6394}
          {arriveTangent=0.2264, interpMode=RCIM_Cubic, leaveTangent=0.2264, tangentMode=RCTM_User, time=0.07, value=1.7344}
          <at 1.0062 1.9434 RCIM_Cubic>
         ]}}})))
   &FN_SpriteRenderer(
    Color:LinearColor=&MultiplyLinearColor(
     Linear Color=$Particles.Initial.Color
     Scale Factor=&MakeFloat4FromVectorAndFloat(
      "Vector (RGB)"=<1 1 1>
      "Float (Alpha)"=&FloatFromCurve(FloatCurve=DataInterface() {propertyValues {Curve {keys [
           <at -0.0038 0.0215 RCIM_Cubic>
           {arriveTangent=0.162, interpMode=RCIM_Cubic, leaveTangent=0.162, tangentMode=RCTM_User, time=0.1558, value=0.7168}
           <at 0.6367 0.29>
           <at 1 0>
          ]}}})
     )
    )
    Create Color Attribute=true
    Create SpriteSize Attribute=true
    SpriteSize:Vector2f=&Multiply_Vector2(A=$Particles.SpriteSize, B=&Vector2DFromFloat(Value=0.8))
   )
   &FN_SpriteRenderer(Camera Offset=100, Create Camera Offset Attribute=true, Create Sprite Rotation Attribute=true) {module=FN_SpriteRenderer001}
  ]
  rendererClasses [&NiagaraSpriteRendererProperties]
  renderers [
   {rendererClass=&NiagaraSpriteRendererProperties}
   {rendererClass=&NiagaraSpriteRendererProperties, rendererIndex=1}
   {rendererClass=&NiagaraSpriteRendererProperties, rendererIndex=2}
   {rendererClass=&NiagaraSpriteRendererProperties, rendererIndex=3}
  ]
 }
}
summary {
 systemName=NS_Emote_HighTowerMango_Burst
 emitters [
  {emitterName=HeavySmoke, rendererClasses [&NiagaraSpriteRendererProperties]}
  {emitterName=LightSmoke, rendererClasses [&NiagaraSpriteRendererProperties]}
  {emitterName=LightSmoke001, rendererClasses [&NiagaraSpriteRendererProperties]}
 ]
 userVariables [
  User.BKSL_CurrentLevel {type=NiagaraFloat, defaultValue=160, description=""}
  User.BKSL_StartAtLevel {type=NiagaraFloat, defaultValue=120, description=""}
  User.BKSL_StopAtLevel {type=NiagaraFloat, defaultValue=160, description=""}
  User.BKSL_VariantColor {type=LinearColor, description="", defaultValue {struct=LinearColor, value=<rgba 1 1 1 1>}}
  User.Delay {type=NiagaraFloat, defaultValue=3.33, description=""}
  User.SkeletalMesh {
   type=NiagaraDataInterfaceSkeletalMesh
   defaultValue {
    struct=NiagaraExt_VariableValue_DataInterface
    value {dataInterface=&self:NiagaraDataInterfaceSkeletalMesh_0}
   }
  }
 ]
}
