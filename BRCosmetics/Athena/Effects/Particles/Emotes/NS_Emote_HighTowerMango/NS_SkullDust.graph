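# Niagara system for the HighTowerMango emote: a red fiery burst of flames, sparks and flare emitted from points on the User.SkeletalMesh.
# Fire rises with shrinking spawn rate and size, while Burst swirls sprites on a widening, climbing torus driven by emitter age.
# Sparks scatter upward with curl noise; Flare_Burst tracks the mesh with a short glow and red particle light.
# The Dusy emitter is disabled.
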NiagaraSystem /BRCosmetics/Athena/Effects/Particles/Emotes/NS_Emote_HighTowerMango/NS_SkullDust
dependencies {
 usedDataInterfaces [
  &NiagaraDataInterfaceVector2DCurve
  &NiagaraDataInterfaceCurve
  &NiagaraDataInterfaceVectorCurve
  &NiagaraDataInterfaceColorCurve
 ]
 usedDynamicInputs []
 usedModules [&SystemState]
}
emitters {
 Burst {
  emitterUpdateScript [
   &EmitterState()
   &SpawnRate(
    SpawnRate=&FloatFromCurve(FloatCurve=DataInterface() {propertyValues {Curve {keys [<at 0 1>,<at 1 0>]}}}, CurveIndex=$Emitter.Age, Scale Curve=25)
    Use Spawn Probability=false
   )
  ]
  particleSpawnScript [
   /Niagara/Modules/Spawn/Initialization/InitializeParticle(
    Color=&MultiplyLinearColor(
     Linear Color=&UniformRangedLinearColor(Minimum=<rgba 20 0 0 1>, Maximum=<rgba 10 0 0 1>)
     Scale Factor=&MakeVector4_FromFloat(Float=&FloatFromCurve(
       FloatCurve=DataInterface() {propertyValues {Curve {keys [{arriveTangent=-2.024, interpMode=RCIM_Cubic, leaveTangent=-2.024, tangentMode=RCTM_User, time=0, value=1}, <at 1 0>]}}}
       CurveIndex=$Emitter.Age
       Scale Curve=0.6
      ))
    )
    Lifetime=&Multiply_Float(
     A=&UniformRangedFloat(Minimum=0.25, Maximum=0.5)
     B=&FloatFromCurve(
      FloatCurve=DataInterface() {propertyValues {Curve {keys [{arriveTangent=-1.7529, interpMode=RCIM_Cubic, leaveTangent=-1.7529, tangentMode=RCTM_User, time=0, value=1}, <at 0.9966 0.2773>]}}}
      CurveIndex=$Emitter.Age
     )
    )
    Module.Write Color=true
    Module.Write Lifetime=true
    Module.Write Mass=true
    Module.Write Material Random=true
    Module.Write Position=false
    Module.Write Scale=false
    Module.Write SpriteRotation=true
    Module.Write SpriteSize=true
    Position=&SimulationPosition()
    Sprite Rotation=&UniformRangedFloat(Minimum=0, Maximum=360)
    Sprite Size=&Multiply_Vector2(
     A=&Vector2DFromFloat(Value=&UniformRangedFloat(Minimum=12, Maximum=24))
     B=&Vector2DFromFloat(Value=&FloatFromCurve(FloatCurve=DataInterface() {propertyValues {Curve {keys [<at 0 1>,<at 1 0>]}}}, CurveIndex=$Emitter.Age))
    )
   )
   /Niagara/Modules/Spawn/Location/SphereLocation(
    MaskSpawnGroup=false
    Module.Override Randomness=false
    Sphere Origin:Vector3f=$Output.SkeletalMeshLocation.SampledPosition
    Sphere Radius=&FloatFromCurve(
     FloatCurve=DataInterface() {propertyValues {Curve {keys [{arriveTangent=-19.3845, interpMode=RCIM_Cubic, leaveTangent=-19.3845, tangentMode=RCTM_User, time=-0.009, value=13.6816}, <at 1 0>]}}}
     CurveIndex=$Emitter.Age
    )
    Surface Only=false
   ) {enabled=false}
   &AddVelocity(
    Coordinate Space=::World
    Velocity=&RotateVectorByQuaternion(
     Vector To Rotate=&Multiply_VectorByFloat(
      Vector=/Game/Effects/Niagara/DynamicInputs/Random/FN_RandomVectorInCone(
       Cone Angle:NiagaraFloat=24
       Cone Axis:Vector3f=<0 1 0.25>
       Cone Axis Coordinate Space:ENiagaraCoordinateSpace=::World
       Velocity Distribution Along Cone Axis:NiagaraFloat=0.5
      )
      Float=/Game/Effects/Niagara/DynamicInputs/Math/FN_Fit01(Clamp:NiagaraBool=false, Value:NiagaraFloat=$Particles.MaterialRandom, DestMin:NiagaraFloat=50, DestMax:NiagaraFloat=48.5728)
     )
     Quaternion=$Output.SkeletalMeshLocation.SampledRotation
    )
   ) {enabled=false}
   &SkeletalMeshLocation(
    Apply Sampled Orientation to Mesh Particles=true
    Bone Sampling Mode=::NewEnumerator4
    Inherit Velocity=false
    MaskSpawnGroup=false
    Sampled Position Offset Coordinate Space:ENiagaraMeshTransforms=::NewEnumerator3
    Skeletal Mesh=$User.SkeletalMesh
   )
   /SetVariables(Particles.RandomV:NiagaraFloat=&UniformRangedFloat(Minimum=-0.4, Maximum=0.4)) {isSetParameters=true}
   /SetVariables(Particles.RandomGravity:NiagaraFloat=&FN_Fit01(Clamp:NiagaraBool=false, Value:NiagaraFloat=0, DestMin:NiagaraFloat=100, DestMax:NiagaraFloat=300)) {isSetParameters=true, module=SetVariables_#1}
  ]
  particleUpdateScript [
   &ParticleState()
   &AccelerationForce(Acceleration=&MakeVector(X=0, Y=0, Z=$Particles.RandomGravity)) {enabled=false}
   /SetVariables(Particles.Radius:NiagaraFloat=&FloatFromCurve(
     FloatCurve=DataInterface() {propertyValues {Curve {keys [<at 0 1 RCIM_Cubic>, {arriveTangent=43.4974, interpMode=RCIM_Cubic, leaveTangent=43.4973, tangentMode=RCTM_User, time=1.0011, value=37.3639}]}}}
     CurveIndex=$Emitter.Age
     Scale Curve=0.75
    )) {isSetParameters=true}
   &TorusLocation(
    Align Torus With Axis=false
    Coordinate Space:ENiagaraCoordinateSpace=::World
    Handle Radius:NiagaraFloat=&Multiply_Float(A=$Particles.Radius, B=0.6)
    Large Radius:NiagaraFloat=$Particles.Radius
    MaskSpawnGroup=false
    Offset=&MakeVector(
     X=0
     Y=0
     Z=&FloatFromCurve(
      FloatCurve=DataInterface() {propertyValues {Curve {keys [
          {arriveTangent=2.3659, interpMode=RCIM_Cubic, leaveTangent=2.3659, tangentMode=RCTM_User, time=-0.0107, value=0.0571}
          {arriveTangent=2.5078, interpMode=RCIM_Cubic, leaveTangent=2.5078, tangentMode=RCTM_User, time=0.9968, value=2.4875}
         ]}}}
      CurveIndex=$Emitter.Age
      Scale Curve=30
     )
    )
    Radius Position:NiagaraFloat=&Multiply_Float(A=$Emitter.Age, B=1)
    Torus Distribution Mode:ENiagaraTorusDistributionMode=::NewEnumerator1
    Torus Origin:NiagaraPosition=$Particles.Initial.Position
    U Position:NiagaraFloat=$Particles.MaterialRandom
    V Position:NiagaraFloat=&Add_Float(A=&Multiply_Float(A=$Emitter.Age, B=-1.5), B=$Particles.Initial.RandomV)
   )
   &SolveForcesAndVelocity(Clamp Velocity=false, Limit Acceleration=false) {enabled=false}
   &Color(
    Color=$Particles.Initial.Color
    Scale Alpha=&FloatFromCurve(FloatCurve=DataInterface() {propertyValues {Curve {keys [
         <at -0.0018 0.0354>
         {arriveTangent=1.3315, interpMode=RCIM_Cubic, leaveTangent=1.3315, time=0.1661, value=0.9253}
         <at 0.7484 1.0343>
         <at 1 0>
        ]}}})
   )
   /Niagara/Modules/Update/SubUV/SubUVAnimation(
    Number Of Frames=64
    SubUV Lookup Index=&Add_Float(
     A=&FN_Fit01(Clamp:NiagaraBool=false, Value:NiagaraFloat=$Particles.MaterialRandom, DestMin:NiagaraFloat=0.2, DestMax:NiagaraFloat=0.3)
     B=&FN_Fit01(Clamp:NiagaraBool=false, Value:NiagaraFloat=$Particles.MaterialRandom, DestMin:NiagaraFloat=0, DestMax:NiagaraFloat=0.4)
    )
   )
   &DynamicMaterialParameters(
    Index 0 Param 1=&FloatFromCurve(FloatCurve=DataInterface() {propertyValues {Curve {keys [<at 0.1692 0.0092 RCIM_Cubic>,<at 0.4923 1.0207 RCIM_Cubic>]}}})
    Index 0 Param 2=&FN_Fit01(Clamp:NiagaraBool=false, Value:NiagaraFloat=$Particles.MaterialRandom, DestMin:NiagaraFloat=1, DestMax:NiagaraFloat=3)
    Index 0 Param 3=&FloatFromCurve(FloatCurve=DataInterface() {propertyValues {Curve {keys [<at 0 1 RCIM_Cubic>,<at 0.9728 1.7662 RCIM_Cubic>]}}})
    Index 0 Param 4=&FN_Fit01(Clamp:NiagaraBool=false, Value:NiagaraFloat=$Particles.MaterialRandom, DestMin:NiagaraFloat=1.0925, DestMax:NiagaraFloat=2.1257)
    Module.Param0WriteEnabled=true
    Module.Param1WriteEnabled=true
    Module.Param2WriteEnabled=true
    Module.Param3WriteEnabled=true
   )
   &ScaleSpriteSize(Scale Factor=&Vector2DFromCurve(
     Vector2Curve=DataInterface() {propertyValues {
       XCurve {keys [<at 0 1 RCIM_Cubic>, {arriveTangent=0.6337, interpMode=RCIM_Cubic, leaveTangent=0.6337, tangentMode=RCTM_User, time=1.0072, value=1.4614}]}
       YCurve {keys [<at 0 1 RCIM_Cubic>, {arriveTangent=0.953, interpMode=RCIM_Cubic, leaveTangent=0.953, tangentMode=RCTM_User, time=1.0101, value=1.4166}]}
      }}
     ScaleCurve=<1 1>
    ))
   &SpriteRotationRate(Rotation Rate=100)
   &CameraOffset(Camera Offset Amount=100)
  ]
  rendererClasses [&NiagaraSpriteRendererProperties]
  renderers [{rendererClass=&NiagaraSpriteRendererProperties}]
 }
 Dusy {
  bEnabled=false
  emitterUpdateScript [
   &EmitterState(Life Cycle Mode=::NewEnumerator1, Loop Behavior=::NewEnumerator1, Loop Delay=1, Loop Duration=1, UseLoopDelay=true)
   &SpawnBurst_Instantaneous(Spawn Count=5, Use Spawn Probability=false)
  ]
  particleSpawnScript [
   &InitializeParticle(
    Color=&UniformRangedLinearColor(Minimum=<rgba 1 0.5 0 0.1>, Maximum=<rgba 0.3 0.1 0 0.2>)
    Module.Write Color=true
    Module.Write Lifetime=true
    Module.Write Mass=true
    Module.Write Material Random=true
    Module.Write Position=false
    Module.Write Scale=false
    Module.Write SpriteRotation=false
    Module.Write SpriteSize=true
    Position=&SimulationPosition()
    Sprite Size=<30 55>
   )
   &InheritVelocity(Module.Limit Speed=true) {enabled=false}
   &SkeletalMeshLocation(
    Apply Sampled Orientation to Mesh Particles=true
    Bone Sampling Mode=::NewEnumerator4
    Inherit Velocity=false
    Interpolate Mesh Sampling=false
    MaskSpawnGroup=false
    Skeletal Mesh=$User.SkeletalMesh
   )
   &AddVelocity(
    Coordinate Space=::World
    Velocity=&RotateVectorByQuaternion(
     Vector To Rotate=&Multiply_VectorByFloat(
      Vector=&FN_RandomVectorInCone(
       Cone Angle:NiagaraFloat=50
       Cone Axis:Vector3f=<0 1 0.25>
       Cone Axis Coordinate Space:ENiagaraCoordinateSpace=::World
       Velocity Distribution Along Cone Axis:NiagaraFloat=0.5
      )
      Float=&FN_Fit01(Clamp:NiagaraBool=false, Value:NiagaraFloat=$Particles.MaterialRandom, DestMin:NiagaraFloat=200, DestMax:NiagaraFloat=250)
     )
     Quaternion=$Output.SkeletalMeshLocation.SampledRotation
    )
   )
   &SubUVAnimation(SubUV Animation Mode=::NewEnumerator1)
  ]
  particleUpdateScript [
   &ParticleState()
   &GravityForce(Gravity=&MakeVector(
     X=0
     Y=0
     Z=&FN_Fit01(Clamp:NiagaraBool=false, Value:NiagaraFloat=$Particles.Initial.MaterialRandom, DestMin:NiagaraFloat=-600, DestMax:NiagaraFloat=-1200)
    ))
   &SolveForcesAndVelocity(Clamp Velocity=false, Limit Acceleration=false)
   &ScaleSpriteSize(Scale Factor=&Vector2DFromCurve(
     Vector2Curve=DataInterface() {propertyValues {
       XCurve {keys [<at -0.0077 0.6287 RCIM_Cubic>,<at 0.9815 0.9706>]}
       YCurve {keys [<at 0 1 RCIM_Cubic>, {arriveTangent=2.0408, interpMode=RCIM_Cubic, leaveTangent=2.0408, tangentMode=RCTM_User, time=0.9718, value=1.8394}]}
      }}
     ScaleCurve=<1 1>
    ))
  ]
  rendererClasses [&NiagaraSpriteRendererProperties]
  renderers [{rendererClass=&NiagaraSpriteRendererProperties}]
 }
 Fire {
  emitterUpdateScript [
   &EmitterState()
   &SpawnRate(
    SpawnRate=&FloatFromCurve(FloatCurve=DataInterface() {propertyValues {Curve {keys [<at 0 1>,<at 1 0>]}}}, CurveIndex=$Emitter.Age, Scale Curve=20)
    Use Spawn Probability=false
   )
  ]
  particleSpawnScript [
   &InitializeParticle(
    Color=&MultiplyLinearColor(
     Linear Color=&UniformRangedLinearColor(Minimum=<rgba 6 0.1 0 1>, Maximum=<rgba 3 0 0 1>)
     Scale Factor=&MakeVector4_FromFloat(Float=&FloatFromCurve(
       FloatCurve=DataInterface() {propertyValues {Curve {keys [{arriveTangent=-2.024, interpMode=RCIM_Cubic, leaveTangent=-2.024, tangentMode=RCTM_User, time=0, value=1}, <at 1 0>]}}}
       CurveIndex=$Emitter.Age
       Scale Curve=0.6
      ))
    )
    Lifetime=&Multiply_Float(
     A=&UniformRangedFloat(Minimum=0.25, Maximum=0.5)
     B=&FloatFromCurve(
      FloatCurve=DataInterface() {propertyValues {Curve {keys [{arriveTangent=-1.7529, interpMode=RCIM_Cubic, leaveTangent=-1.7529, tangentMode=RCTM_User, time=0, value=1}, <at 0.9966 0.2773>]}}}
      CurveIndex=$Emitter.Age
     )
    )
    Module.Write Color=true
    Module.Write Lifetime=true
    Module.Write Mass=true
    Module.Write Material Random=true
    Module.Write Position=false
    Module.Write Scale=false
    Module.Write SpriteRotation=true
    Module.Write SpriteSize=true
    Position=&SimulationPosition()
    Sprite Rotation=&UniformRangedFloat(Minimum=0, Maximum=360)
    Sprite Size=&Multiply_Vector2(
     A=&Vector2DFromFloat(Value=&UniformRangedFloat(Minimum=24, Maximum=48))
     B=&Vector2DFromFloat(Value=&FloatFromCurve(FloatCurve=DataInterface() {propertyValues {Curve {keys [<at 0 1>,<at 1 0>]}}}, CurveIndex=$Emitter.Age, Scale Curve=1.2))
    )
   )
   &SkeletalMeshLocation(
    Apply Sampled Orientation to Mesh Particles=true
    Bone Sampling Mode=::NewEnumerator4
    Inherit Velocity=false
    Interpolate Mesh Sampling=false
    MaskSpawnGroup=false
    Sampled Position Offset Coordinate Space:ENiagaraMeshTransforms=::NewEnumerator3
    Skeletal Mesh=$User.SkeletalMesh
   )
   &SphereLocation(
    MaskSpawnGroup=false
    Module.Override Randomness=false
    Sphere Origin:Vector3f=$Output.SkeletalMeshLocation.SampledPosition
    Sphere Radius=&FloatFromCurve(
     FloatCurve=DataInterface() {propertyValues {Curve {keys [{arriveTangent=-19.3845, interpMode=RCIM_Cubic, leaveTangent=-19.3845, tangentMode=RCTM_User, time=-0.009, value=13.6816}, <at 1 0>]}}}
     CurveIndex=$Emitter.Age
    )
    Surface Only=false
   )
   &AddVelocity(
    Coordinate Space=::World
    Velocity=&RotateVectorByQuaternion(
     Vector To Rotate=&Multiply_VectorByFloat(
      Vector=&FN_RandomVectorInCone(
       Cone Angle:NiagaraFloat=24
       Cone Axis:Vector3f=<0 0 1>
       Cone Axis Coordinate Space:ENiagaraCoordinateSpace=::World
       Velocity Distribution Along Cone Axis:NiagaraFloat=0.5
      )
      Float=&FN_Fit01(Clamp:NiagaraBool=false, Value:NiagaraFloat=$Particles.MaterialRandom, DestMin:NiagaraFloat=50, DestMax:NiagaraFloat=48.5728)
     )
     Quaternion=$Output.SkeletalMeshLocation.SampledRotation
    )
   )
   /SetVariables(Particles.RandomGravity:NiagaraFloat=&FN_Fit01(Clamp:NiagaraBool=false, Value:NiagaraFloat=0, DestMin:NiagaraFloat=100, DestMax:NiagaraFloat=300)) {isSetParameters=true, module=SetVariables_#1}
  ]
  particleUpdateScript [
   &ParticleState()
   &AccelerationForce(Acceleration=&MakeVector(X=0, Y=0, Z=$Particles.RandomGravity))
   &SolveForcesAndVelocity(Clamp Velocity=false, Limit Acceleration=false)
   &Color(
    Color=$Particles.Initial.Color
    Scale Alpha=&FloatFromCurve(FloatCurve=DataInterface() {propertyValues {Curve {keys [<at 0 1.2289>, {arriveTangent=-1.2289, interpMode=RCIM_Cubic, leaveTangent=-1.2289, time=0.8156, value=1.1101}, <at 1 0>]}}})
   )
   &SubUVAnimation(Number Of Frames=65)
   &DynamicMaterialParameters(
    Index 0 Param 1=&FloatFromCurve(FloatCurve=DataInterface() {propertyValues {bOptimizeLUT=false, Curve {keys [<at 0.1692 0.0092 RCIM_Cubic>,<at 0.4923 1.0207 RCIM_Cubic>]}}})
    Index 0 Param 2=&FN_Fit01(Clamp:NiagaraBool=false, Value:NiagaraFloat=$Particles.MaterialRandom, DestMin:NiagaraFloat=1, DestMax:NiagaraFloat=3)
    Index 0 Param 3=&FloatFromCurve(FloatCurve=DataInterface() {propertyValues {Curve {keys [<at 0 1 RCIM_Cubic>,<at 0.9728 1.7662 RCIM_Cubic>]}}})
    Index 0 Param 4=&FN_Fit01(Clamp:NiagaraBool=false, Value:NiagaraFloat=$Particles.MaterialRandom, DestMin:NiagaraFloat=1.0925, DestMax:NiagaraFloat=2.1257)
    Module.Param0WriteEnabled=true
    Module.Param1WriteEnabled=true
    Module.Param2WriteEnabled=true
    Module.Param3WriteEnabled=true
   ) {module=DynamicMaterialParameters001}
   &ScaleSpriteSize(Scale Factor=&Vector2DFromCurve(
     Vector2Curve=DataInterface() {propertyValues {
       XCurve {keys [<at 0 1 RCIM_Cubic>,<at 0.9975 0.9839>]}
       YCurve {keys [<at 0 1 RCIM_Cubic>, {arriveTangent=2.9123, interpMode=RCIM_Cubic, leaveTangent=2.9123, tangentMode=RCTM_User, time=0.9874, value=2.0457}]}
      }}
     ScaleCurve=<1 1>
    ))
   &CameraOffset(Camera Offset Amount=&FN_Fit01(Clamp:NiagaraBool=false, Value:NiagaraFloat=$Particles.MaterialRandom, DestMin:NiagaraFloat=20, DestMax:NiagaraFloat=30))
  ]
  rendererClasses [&NiagaraSpriteRendererProperties]
  renderers [{rendererClass=&NiagaraSpriteRendererProperties}]
 }
 Flare_Burst {
  emitterUpdateScript [
   &EmitterState(Life Cycle Mode=::NewEnumerator1, Loop Behavior=::NewEnumerator1, Loop Duration=1, UseLoopDelay=false)
   &SpawnBurst_Instantaneous(Spawn Count=1, Use Spawn Probability=false)
   &SpawnBurst_Instantaneous(Spawn Count=1, Spawn Group=1, Use Spawn Probability=false) {enabled=false, module=SpawnBurst_Instantaneous001}
  ]
  particleSpawnScript [
   &InitializeParticle(
    Color=<rgba 1 1 1 25>
    Lifetime=0.4
    Mesh Scale=<1.3 1.3 1.3>
    Module.Write Color=true
    Module.Write Lifetime=true
    Module.Write Mass=true
    Module.Write Material Random=true
    Module.Write Position=false
    Module.Write Scale=false
    Module.Write SpriteRotation=true
    Module.Write SpriteSize=true
    Position=&SimulationPosition()
    Sprite Rotation=&UniformRangedFloat(Minimum=-1, Maximum=1)
    Sprite Size=&Multiply_Vector2(A=<140 1>, B=&Vector2DFromFloat(Value=1))
   )
   &SystemLocation(MaskSpawnGroup=false, Offset=<2.5 -11 -2>) {enabled=false}
   &SystemLocation(MaskSpawnGroup=true, Offset=<-2.5 -11 -2>, Spawn Group Mask=1) {enabled=false, module=SystemLocation001}
   /SetVariables(Particles.Position=&Add_Vector(
     A=Linked() {linkedVariable {name=Particles.Position, type=NiagaraPosition}}
     B=&TransformVector(Vector=$User.Offset, Source Space=::World, Destination Space=::Local)
    )) {enabled=false, isSetParameters=true}
   &SubUVAnimation(SubUV Animation Mode=::NewEnumerator1)
  ]
  particleUpdateScript [
   &ParticleState()
   &SkeletalMeshLocation(
    Apply Sampled Orientation to Mesh Particles=true
    Bone / Socket Index=&SpawnGroup()
    Bone Sampling Mode=::NewEnumerator4
    Inherit Velocity=false
    Interpolate Mesh Sampling=false
    MaskSpawnGroup=false
    Sampled Position Offset=&MakeVector(X=-3.5, Y=-7, Z=-0.5)
    Sampled Position Offset Coordinate Space:ENiagaraMeshTransforms=::NewEnumerator3
    Skeletal Mesh=$User.SkeletalMesh
   ) {module=SkeletalMeshLocation001}
   &ScaleSpriteSize(Scale Factor=&Vector2DFromFloat(Value=&FloatFromCurve(FloatCurve=DataInterface() {propertyValues {Curve {keys [
          {interpMode=RCIM_Cubic, leaveTangent=-1.4318, tangentMode=RCTM_Break, time=0, value=0}
          {interpMode=RCIM_Cubic, leaveTangent=-3.1122, tangentMode=RCTM_Break, time=0.05, value=1.2}
          {arriveTangent=-0.0976, interpMode=RCIM_Cubic, leaveTangent=-0.0976, tangentMode=RCTM_User, time=1, value=0}
         ]}}})))
   &ScaleColor(
    Scale Alpha=&FloatFromCurve(FloatCurve=DataInterface() {propertyValues {Curve {keys [<at 0.7 1>,<at 1 0>]}}})
    Scale RGB=&MakeVectorFromLinearColorRGB(Color=&ColorFromCurve(ColorCurve=DataInterface() {propertyValues {
        AlphaCurve {keys [<at 0 1 RCIM_Cubic>]}
        BlueCurve {keys [{arriveTangent=-0.0502, interpMode=RCIM_Cubic, time=0.2, value=0}, <at 1 0 RCIM_Cubic>]}
        GreenCurve {keys [{arriveTangent=-0.5017, interpMode=RCIM_Cubic, time=0.2, value=0}, <at 1 0 RCIM_Cubic>]}
        RedCurve {keys [{arriveTangent=-5.0168, interpMode=RCIM_Cubic, time=0.2, value=5}, <at 1 0 RCIM_Cubic>]}
       }}))
    ScaleA=true
    ScaleRGB=true
   )
   &CameraOffset(Camera Offset Amount=2.5)
   &SolveForcesAndVelocity(Clamp Velocity=false, Limit Acceleration=false) {enabled=false}
   /Game/Effects/Niagara/Modules/Sprite/FN_SpriteRenderer(
    Camera Offset:NiagaraFloat=&Add_Float(A=$Particles.CameraOffset, B=2)
    Color:LinearColor=<rgba 1 0.5 0 1>
    Create Camera Offset Attribute=true
    Create Color Attribute=true
    Create SpriteSize Attribute=true
    SpriteSize:Vector2f=<20 20>
   )
   &FN_Light_v3(
    Color=<rgba 1000 0 0 1>
    Create Color Attribute=true
    Create Radius Attribute=true
    Enable Particle Light Enable=true
    Light Enabled=HlslExpression() {hlslExpression="Particles.UniqueID == 0 ? 1 : 0"}
    Radius=50
    Scale Color:NiagaraFloat=&FloatFromCurve(FloatCurve=DataInterface() {propertyValues {Curve {keys [<at 0 1>,<at 1 0>]}}})
   )
   &FN_SpriteRenderer(
    Create Position Attribute=true
    Position:Vector3f=&Add_Vector(
     A=Linked() {linkedVariable {name=Particles.Position, type=NiagaraPosition}}
     B=&RotateVectorByQuaternion(Vector To Rotate=<7 0 0>, Quaternion=$Output.SkeletalMeshLocation001.SampledRotation)
    )
   ) {module=FN_SpriteRenderer001}
  ]
  rendererClasses [&NiagaraSpriteRendererProperties, &NiagaraLightRendererProperties]
  renderers [
   {rendererClass=&NiagaraSpriteRendererProperties}
   {rendererClass=&NiagaraSpriteRendererProperties, rendererIndex=1}
   {rendererClass=&NiagaraLightRendererProperties, rendererIndex=2}
   {rendererClass=&NiagaraSpriteRendererProperties, rendererIndex=3}
   {rendererClass=&NiagaraSpriteRendererProperties, rendererIndex=4}
  ]
 }
 Sparks {
  emitterUpdateScript [
   &EmitterState()
   &SpawnBurst_Instantaneous(Spawn Count=50, Use Spawn Probability=false)
  ]
  particleSpawnScript [
   &InitializeParticle(
    Color=&UniformRangedLinearColor(Minimum=<rgba 3 0 0 1>, Maximum=<rgba 1 0 0 1>)
    Lifetime=&UniformAOrBFloat(A=&UniformRangedFloat(Minimum=0.4, Maximum=0.8), B=0.2)
    Module.Write Color=true
    Module.Write Lifetime=true
    Module.Write Mass=true
    Module.Write Material Random=true
    Module.Write Position=false
    Module.Write Scale=false
    Module.Write SpriteRotation=false
    Module.Write SpriteSize=true
    Position=&SimulationPosition()
    Sprite Size=&Multiply_Vector2(
     A=<0.5 2>
     B=&Vector2DFromFloat(Value=&UniformAOrBFloat(A=2, B=0.75))
    )
   )
   &SkeletalMeshLocation(
    Apply Sampled Orientation to Mesh Particles=true
    Bone Sampling Mode=::NewEnumerator4
    Inherit Velocity=false
    Interpolate Mesh Sampling=false
    MaskSpawnGroup=false
    Sampled Position Offset Coordinate Space:ENiagaraMeshTransforms=::NewEnumerator3
    Skeletal Mesh=$User.SkeletalMesh
   )
   &AddVelocity(
    Coordinate Space=::World
    Velocity=&Multiply_VectorByFloat(
     Vector=&FN_RandomVectorInCone(
      Cone Angle:NiagaraFloat=25.9107
      Cone Axis:Vector3f=<0 0 1>
      Cone Axis Coordinate Space:ENiagaraCoordinateSpace=::World
      Velocity Distribution Along Cone Axis:NiagaraFloat=0.5
     )
     Float=&FN_Fit01(Clamp:NiagaraBool=false, Value:NiagaraFloat=$Particles.MaterialRandom, DestMin:NiagaraFloat=18.4732, DestMax:NiagaraFloat=200)
    )
   )
  ]
  particleUpdateScript [
   &ParticleState()
   /Niagara/Modules/Update/Forces/CurlNoiseForce(
    Noise Frequency=0.075
    Noise Strength=&FloatFromCurve(
     FloatCurve=DataInterface() {propertyValues {Curve {keys [{arriveTangent=-1.1841, interpMode=RCIM_Cubic, leaveTangent=-1.1841, tangentMode=RCTM_User, time=0, value=1}, <at 0.9976 0.4523>]}}}
     Scale Curve=600
    )
    Pan Noise=false
   )
   &Drag(Drag=3, Use Linear Drag=true, Use Rotational Drag=false)
   &AccelerationForce(Acceleration=<0 0 20>)
   &SolveForcesAndVelocity(Clamp Velocity=false, Limit Acceleration=false)
   &CalculateAccurateVelocity()
   &ScaleSpriteSize(Scale Factor=&Vector2DFromFloat(Value=&FloatFromCurve(FloatCurve=DataInterface() {propertyValues {Curve {keys [{arriveTangent=-2.0479, interpMode=RCIM_Cubic, time=0.2045, value=1.006}, <at 0.9986 0.2635>]}}})))
   &ScaleColor(
    Scale RGB=&ScaleFloatByCurve(
     InitialValue:Vector3f=<1 1 1>
     ScaleCurve:NiagaraDataInterfaceVectorCurve=DataInterface() {propertyValues {
       XCurve {keys [<at -0.0067 0.9517 RCIM_Cubic>,<at 1.0099 0.9344>]}
       YCurve {keys [<at -0.0109 0.9431 RCIM_Cubic>,<at 0.9922 0.121>]}
       ZCurve {keys [<at -0.0251 0.9583 RCIM_Cubic>,<at 0.9865 0.4544>]}
      }}
     CurveIndex:NiagaraFloat=$Particles.NormalizedAge
    )
    ScaleA=true
    ScaleRGB=true
   )
  ]
  rendererClasses [&NiagaraSpriteRendererProperties]
  renderers [{rendererClass=&NiagaraSpriteRendererProperties}]
 }
}
summary {
 systemName=NS_SkullDust
 emitters [
  {bEnabled=false, emitterName=Dusy, rendererClasses [&NiagaraSpriteRendererProperties]}
  {emitterName=Fire, rendererClasses [&NiagaraSpriteRendererProperties]}
  {emitterName=Burst, rendererClasses [&NiagaraSpriteRendererProperties]}
  {emitterName=Sparks, rendererClasses [&NiagaraSpriteRendererProperties]}
  {emitterName=Flare_Burst, rendererClasses [&NiagaraSpriteRendererProperties, &NiagaraLightRendererProperties]}
 ]
 userVariables [
  User.Offset {type=Vector3f, description="", defaultValue {struct=Vector3f, value=<0 0 0>}}
  User.RibbonIDOffset {type=NiagaraInt32}
  User.SkeletalMesh {
   type=NiagaraDataInterfaceSkeletalMesh
   defaultValue {
    struct=NiagaraExt_VariableValue_DataInterface
    value {dataInterface=&self:NiagaraDataInterfaceSkeletalMesh_0}
   }
  }
 ]
}
