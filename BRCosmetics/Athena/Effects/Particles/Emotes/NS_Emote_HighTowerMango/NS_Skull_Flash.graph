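# Niagara system for the HighTowerMango emote's skull flash: builds red fire and a pair of glowing flares on a user-supplied skeletal mesh.
# Fire spawns flipbook sprites around a sampled mesh point, with rate, size, brightness and radius growing with emitter age as they rise.
# Flare_M pins two mirrored sprites to the mesh, fading them in with accelerating flicker from age-squared curve lookup.
# One flare also drives a brightening red light.
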NiagaraSystem /BRCosmetics/Athena/Effects/Particles/Emotes/NS_Emote_HighTowerMango/NS_Skull_Flash
dependencies {
 usedDataInterfaces [&NiagaraDataInterfaceCurve, &NiagaraDataInterfaceColorCurve, &NiagaraDataInterfaceVector2DCurve]
 usedDynamicInputs []
 usedModules [&SystemState]
}
emitters {
 Fire {
  emitterUpdateScript [
   &EmitterState()
   &SpawnRate(
    SpawnRate=&FloatFromCurve(FloatCurve=DataInterface() {propertyValues {Curve {keys [<at 1.0164 0.0235>,<at 2.2643 0.6366>]}}}, CurveIndex=$Emitter.Age, Scale Curve=10)
    Use Spawn Probability=false
   )
  ]
  particleSpawnScript [
   /Niagara/Modules/Spawn/Initialization/InitializeParticle(
    Color=&MultiplyLinearColor(
     Linear Color=&UniformRangedLinearColor(Minimum=<rgba 6 0.1 0 1>, Maximum=<rgba 3 0 0 1>)
     Scale Factor=&MakeVector4_FromFloat(Float=&FloatFromCurve(
       FloatCurve=DataInterface() {propertyValues {Curve {keys [{arriveTangent=0.2052, interpMode=RCIM_Cubic, leaveTangent=0.2052, tangentMode=RCTM_User, time=0.0276, value=0.0279}, <at 1.9811 1.7776>]}}}
       CurveIndex=$Emitter.Age
       Scale Curve=0.6
      ))
    )
    Lifetime=&Multiply_Float(
     A=&UniformRangedFloat(Minimum=0.25, Maximum=0.5)
     B=&FloatFromCurve(
      FloatCurve=DataInterface() {propertyValues {Curve {keys [{arriveTangent=0.081, interpMode=RCIM_Cubic, leaveTangent=0.081, tangentMode=RCTM_User, time=-0.0039, value=0.2656}, <at 1.2346 0.6182>]}}}
      CurveIndex=$Emitter.Age
      Scale Curve=0.5
     )
    )
    Module.Write Color=true
    Module.Write Lifetime=true
    Module.Write Mass=true
    Module.Write Material Random=true
    Module.Write Position=false
    Module.Write Scale=false
    Module.Write SpriteRotation=true
    Module.Write SpriteSize=true
    Position=&SimulationPosition()
    Sprite Rotation=&UniformRangedFloat(Minimum=0, Maximum=360)
    Sprite Size=&Multiply_Vector2(
     A=&Vector2DFromFloat(Value=&UniformRangedFloat(Minimum=24, Maximum=48))
     B=&Vector2DFromFloat(Value=&FloatFromCurve(
       FloatCurve=DataInterface() {propertyValues {Curve {keys [
           <at -0.0194 0.5427 RCIM_Cubic>
           {arriveTangent=1.6657, interpMode=RCIM_Cubic, leaveTangent=1.6657, tangentMode=RCTM_User, time=2.0079, value=2.0019}
          ]}}}
       CurveIndex=$Emitter.Age
       Scale Curve=0.1
      ))
    )
   )
   &SkeletalMeshLocation(
    Apply Sampled Orientation to Mesh Particles=true
    Bone Sampling Mode=::NewEnumerator4
    Inherit Velocity=false
    Interpolate Mesh Sampling=false
    MaskSpawnGroup=false
    Sampled Position Offset Coordinate Space:ENiagaraMeshTransforms=::NewEnumerator3
    Skeletal Mesh=$User.SkeletalMesh
   )
   /Niagara/Modules/Spawn/Location/SphereLocation(
    MaskSpawnGroup=false
    Module.Override Randomness=false
    Sphere Origin:Vector3f=$Output.SkeletalMeshLocation.SampledPosition
    Sphere Radius=&FloatFromCurve(
     FloatCurve=DataInterface() {propertyValues {Curve {keys [{arriveTangent=-0.2847, interpMode=RCIM_Cubic, leaveTangent=-0.2848, tangentMode=RCTM_User, time=-0.0066, value=2.7194}, <at 1.0029 11.5814>]}}}
     CurveIndex=$Emitter.Age
    )
    Surface Only=false
   )
   &AddVelocity(
    Coordinate Space=::World
    Velocity=&RotateVectorByQuaternion(
     Vector To Rotate=&Multiply_VectorByFloat(
      Vector=/Game/Effects/Niagara/DynamicInputs/Random/FN_RandomVectorInCone(
       Cone Angle:NiagaraFloat=24
       Cone Axis:Vector3f=<0 0 1>
       Cone Axis Coordinate Space:ENiagaraCoordinateSpace=::World
       Velocity Distribution Along Cone Axis:NiagaraFloat=0.5
      )
      Float=/Game/Effects/Niagara/DynamicInputs/Math/FN_Fit01(Clamp:NiagaraBool=false, Value:NiagaraFloat=$Particles.MaterialRandom, DestMin:NiagaraFloat=50, DestMax:NiagaraFloat=48.5728)
     )
     Quaternion=$Output.SkeletalMeshLocation.SampledRotation
    )
   )
   /SetVariables(Particles.RandomGravity:NiagaraFloat=&FN_Fit01(Clamp:NiagaraBool=false, Value:NiagaraFloat=0, DestMin:NiagaraFloat=100, DestMax:NiagaraFloat=300)) {isSetParameters=true}
  ]
  particleUpdateScript [
   &ParticleState()
   &AccelerationForce(Acceleration=&MakeVector(X=0, Y=0, Z=$Particles.RandomGravity))
   &SolveForcesAndVelocity(Clamp Velocity=false, Limit Acceleration=false)
   &Color(
    Color=$Particles.Initial.Color
    Scale Alpha=&FloatFromCurve(FloatCurve=DataInterface() {propertyValues {Curve {keys [<at 0 1.2289>, {arriveTangent=-1.2289, interpMode=RCIM_Cubic, leaveTangent=-1.2289, time=0.8156, value=1.1101}, <at 1 0>]}}})
   )
   /Niagara/Modules/Update/SubUV/SubUVAnimation(Number Of Frames=65)
   &DynamicMaterialParameters(
    Index 0 Param 1=&FloatFromCurve(FloatCurve=DataInterface() {propertyValues {bOptimizeLUT=false, Curve {keys [<at 0.1692 0.0092 RCIM_Cubic>,<at 0.4923 1.0207 RCIM_Cubic>]}}})
    Index 0 Param 2=&FN_Fit01(Clamp:NiagaraBool=false, Value:NiagaraFloat=$Particles.MaterialRandom, DestMin:NiagaraFloat=1, DestMax:NiagaraFloat=3)
    Index 0 Param 3=&FloatFromCurve(FloatCurve=DataInterface() {propertyValues {Curve {keys [<at 0 1 RCIM_Cubic>,<at 0.9728 1.7662 RCIM_Cubic>]}}})
    Index 0 Param 4=&FN_Fit01(Clamp:NiagaraBool=false, Value:NiagaraFloat=$Particles.MaterialRandom, DestMin:NiagaraFloat=1.0925, DestMax:NiagaraFloat=2.1257)
    Module.Param0WriteEnabled=true
    Module.Param1WriteEnabled=true
    Module.Param2WriteEnabled=true
    Module.Param3WriteEnabled=true
   ) {module=DynamicMaterialParameters001}
   &ScaleSpriteSize(Scale Factor=&Vector2DFromCurve(
     Vector2Curve=DataInterface() {propertyValues {
       XCurve {keys [<at 0 1 RCIM_Cubic>,<at 0.9975 0.9839>]}
       YCurve {keys [<at 0 1 RCIM_Cubic>, {arriveTangent=2.9123, interpMode=RCIM_Cubic, leaveTangent=2.9123, tangentMode=RCTM_User, time=0.9874, value=2.0457}]}
      }}
     ScaleCurve=<1 1>
    ))
   &CameraOffset(Camera Offset Amount=&FN_Fit01(Clamp:NiagaraBool=false, Value:NiagaraFloat=$Particles.MaterialRandom, DestMin:NiagaraFloat=20, DestMax:NiagaraFloat=30))
  ]
  rendererClasses [&NiagaraSpriteRendererProperties]
  renderers [{rendererClass=&NiagaraSpriteRendererProperties}]
 }
 Flare_M {
  emitterUpdateScript [
   &EmitterState()
   &SpawnBurst_Instantaneous(Spawn Count=1, Use Spawn Probability=false)
   &SpawnBurst_Instantaneous(Spawn Count=1, Spawn Group=1, Use Spawn Probability=false) {module=SpawnBurst_Instantaneous001}
  ]
  particleSpawnScript [
   &InitializeParticle(
    Color=<rgba 1 0 0 1>
    Lifetime=5
    Mesh Scale=<1.3 1.3 1.3>
    Module.Write Color=true
    Module.Write Lifetime=true
    Module.Write Mass=true
    Module.Write Material Random=true
    Module.Write Position=false
    Module.Write Scale=false
    Module.Write SpriteRotation=false
    Module.Write SpriteSize=true
    Position=&SimulationPosition()
    Sprite Size=&Vector2DFromFloat(Value=22.5)
   )
   &SystemLocation(MaskSpawnGroup=false, Offset=<2.5 -11 -2>) {enabled=false}
   &SystemLocation(MaskSpawnGroup=true, Offset=<-2.5 -11 -2>, Spawn Group Mask=1) {enabled=false, module=SystemLocation001}
   /SetVariables() {isSetParameters=true}
  ]
  particleUpdateScript [
   &ParticleState(Kill Particles When Lifetime Has Elapsed=false, Let Infinitely Lived Particles Die When Emitter Deactivates=true)
   &SkeletalMeshLocation(
    Apply Sampled Orientation to Mesh Particles=true
    Bone / Socket Index=&SpawnGroup()
    Bone Sampling Mode=::NewEnumerator4
    Inherit Velocity=false
    Interpolate Mesh Sampling=false
    MaskSpawnGroup=false
    Sampled Position Offset=&MakeVector(X=&MaskFloatBySpawnGroup(Masked Float=-3.5, Passthrough Float=3.5), Y=-7, Z=-0.5)
    Sampled Position Offset Coordinate Space:ENiagaraMeshTransforms=::NewEnumerator3
    Skeletal Mesh=$User.SkeletalMesh
   ) {module=SkeletalMeshLocation001}
   &ScaleSpriteSize(Scale Factor=&Vector2DFromFloat(Value=&FloatFromCurve(
      FloatCurve=DataInterface() {propertyValues {Curve {keys [<at 0 1>,<at 0.1341 0.726>,<at 0.2151 1.053>,<at 0.4084 0.5013>,<at 0.684 1.1084>,<at 0.8681 0.6573>,<at 0.9922 1>]}}}
      CurveIndex=HlslExpression() {hlslExpression="Modulo((Emitter.Age*30*Emitter.Age),1.0)"}
      Scale Curve=&FloatFromCurve(
       FloatCurve=DataInterface() {propertyValues {Curve {keys [<at 0 1 RCIM_Cubic>, {arriveTangent=2.1847, interpMode=RCIM_Cubic, leaveTangent=2.1847, tangentMode=RCTM_User, time=0.9934, value=2.0576}]}}}
       CurveIndex=$Emitter.NormalizedLoopAge
      )
     )))
   &ScaleColor(
    Scale Alpha=&FloatFromCurve(
     FloatCurve=DataInterface() {propertyValues {Curve {keys [
         {arriveTangent=0.162, interpMode=RCIM_Cubic, leaveTangent=0.162, tangentMode=RCTM_User, time=-0.0269, value=0}
         {arriveTangent=0.6627, interpMode=RCIM_Cubic, leaveTangent=0.6627, tangentMode=RCTM_User, time=2.9909, value=0.9}
        ]}}}
     CurveIndex=$Emitter.Age
    )
    Scale RGB=&MakeVectorFromLinearColorRGB(Color=&ColorFromCurve(ColorCurve=DataInterface() {propertyValues {
        AlphaCurve {keys [<at 0 1 RCIM_Cubic>]}
        BlueCurve {keys [<at 0.2 0 RCIM_Cubic>]}
        GreenCurve {keys [<at 0.2 0 RCIM_Cubic>]}
        RedCurve {keys [<at 0.2 1 RCIM_Cubic>]}
       }}))
    ScaleA=true
    ScaleRGB=true
   )
   &CameraOffset(Camera Offset Amount=2)
   &FN_Light_v3(
    Color=&MakeLinearColor(
     R:NiagaraFloat=&FloatFromCurve(
      FloatCurve=DataInterface() {propertyValues {Curve {keys [
          {arriveTangent=-0.0281, interpMode=RCIM_Cubic, leaveTangent=-0.0281, tangentMode=RCTM_User, time=1.093, value=0.0029}
          {arriveTangent=0.6627, interpMode=RCIM_Cubic, leaveTangent=0.6627, tangentMode=RCTM_User, time=2.9909, value=0.9}
         ]}}}
      CurveIndex=$Emitter.Age
      Scale Curve=400
     )
     G:NiagaraFloat=0
     B:NiagaraFloat=0
     A:NiagaraFloat=0
    )
    Create Color Attribute=true
    Create Radius Attribute=true
    Enable Particle Light Enable=true
    Light Enabled=HlslExpression() {hlslExpression="Particles.UniqueID == 0"}
    Radius=200
   )
   &SolveForcesAndVelocity(Clamp Velocity=false, Limit Acceleration=false) {enabled=false}
  ]
  rendererClasses [&NiagaraSpriteRendererProperties, &NiagaraLightRendererProperties]
  renderers [
   {rendererClass=&NiagaraSpriteRendererProperties}
   {rendererClass=&NiagaraLightRendererProperties, rendererIndex=1}
  ]
 }
}
summary {
 systemName=NS_Skull_Flash
 emitters [
  {emitterName=Flare_M, rendererClasses [&NiagaraSpriteRendererProperties, &NiagaraLightRendererProperties]}
  {emitterName=Fire, rendererClasses [&NiagaraSpriteRendererProperties]}
 ]
 userVariables [
  User.Offset {type=Vector3f, description="", defaultValue {struct=Vector3f, value=<0 0 27.4512>}}
  User.Right2ImpactVelocity {type=Vector3f, description="", defaultValue {struct=Vector3f, value=<0 0 0>}}
  User.SkeletalMesh {
   type=NiagaraDataInterfaceSkeletalMesh
   defaultValue {
    struct=NiagaraExt_VariableValue_DataInterface
    value {dataInterface=&self:NiagaraDataInterfaceSkeletalMesh_0}
   }
  }
 ]
}
